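# Translucent, lit water material for the Zero Point center wave in the Hera environment, with refraction and animated vertex displacement.
# A custom HLSL node sums six Gerstner waves into offset, normal and per-wave masks, driving edge-masked displacement along vertex normals.
# Height and wave masks blend water colors and tinted, panning foam.
# Height, depth, camera distance and view angle shape opacity, while platform switches drop sky-light reflection and desaturate color on weaker targets.
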Material /Game/Environments/Hera/FX/ZeroPoint/WaterDisturbance/M_Env_ZeroPoint_CenterWaterWave_Optimized
graph {
 properties {
  bAutomaticallySetUsageInEditor=false
  blendMode=BLEND_Translucent
  previewMesh=&/SM_WaveMesh_Cut
  refractionMethod=RM_IndexOfRefraction
  translucencyLightingMode=TLM_VolumetricDirectional
  usedShadingModels=Default Lit
 }
 nodes [
  MaterialExpressionMultiply_26 = Multiply(A=TextureCoordinate(), B=ScalarParameter() {defaultValue=1437.7084, group=01_Wave, parameterName=WaveScale})
  MaterialExpressionObjectPositionWS_1 = ObjectPositionWS()
  n1 = Convert(None=@MaterialExpressionObjectPositionWS_1) {
   nodeName=Break Float3
   convertInputs [{
     type=Vector3
     expressionInput {
      expression=&self:MaterialExpressionObjectPositionWS_1
      inputName=None
      mask=0
      maskA=0
      maskB=0
      maskG=0
      maskR=0
      outputIndex=0
     }
    }]
   convertOutputs [{type=Scalar}, {type=Scalar}, {type=Scalar}]
  }
  n2 = ScalarParameter() {group=99_Misc, parameterName=TR}
  MaterialExpressionMultiply_27 = Multiply(
   A=Add(A=Time() {bOverride_Period=true, period=6000}, B=Multiply(A=@n1, B=@n2))
   B=ScalarParameter() {defaultValue=1.5, group=01_Wave, parameterName=WaveTimeScale}
  )
  MaterialExpressionReroute_181 = Reroute(None=ScalarParameter() {defaultValue=1.1, group=01_Wave, parameterName=Steepness})
  MaterialExpressionReroute_182 = Reroute(None=ScalarParameter() {defaultValue=0.37, group=01_Wave, parameterName=Amplitude})
  MaterialExpressionReroute_196 = Reroute(None=ScalarParameter() {defaultValue=50, group=01_Wave, parameterName=NormalStrength})
  n3 = Custom!(
   Position=@MaterialExpressionMultiply_26
   Time=@MaterialExpressionMultiply_27
   Steepness=@MaterialExpressionReroute_181
   Amplitude=@MaterialExpressionReroute_182
   NormalStrength=@MaterialExpressionReroute_196
  ) {
   code="float2 dirs[6];\r\ndirs[0] = normalize(float2( 1.00,  0.20));\r\ndirs[1] = normalize(float2( 0.45,  0.89));\r\ndirs[2] = normalize(float2(-0.60,  0.80));\r\ndirs[3] = normalize(float2(-0.92, -0.39));\r\ndirs[4] = normalize(float2( 0.18, -0.98));\r\ndirs[5] = normalize(float2( 0.78, -0.62));\r\nfloat wavelengths[6] = { 1500.0, 900.0, 520.0, 300.0, 170.0, 95.0 };\r\nfloat amps[6]        = {  1.00,  0.62,  0.40,  0.26,  0.16,  0.10 };\r\n\r\nfloat3 offset = float3(0, 0, 0);\r\nfloat Nx = 0.0;\r\nfloat Ny = 0.0;\r\nfloat Nz = 1.0;\r\nfloat waveGray[6];\r\n\r\n[unroll]\r\nfor (int i = 0; i < 6; i++)\r\n{\r\n    float A   = Amplitude * amps[i];\r\n    float w   = 6.28318530718 / wavelengths[i];\r\n    float phi = sqrt(9.81 * w * 100.0);\r\n    float Q   = Steepness / (w * A * 6.0);\r\n    float theta = dot(dirs[i], Position.xy) * w + Time * phi;\r\n    float c = cos(theta);\r\n    float s = sin(theta);\r\n\r\n    offset.x += Q * A * dirs[i].x * c;\r\n    offset.y += Q * A * dirs[i].y * c;\r\n    offset.z += A * s;\r\n\r\n    float WA = w * A;\r\n    Nx -= dirs[i].x * WA * c;\r\n    Ny -= dirs[i].y * WA * c;\r\n    Nz -= Q * WA * s;\r\n\r\n    waveGray[i] = s * 0.5 + 0.5;\r\n}\r\n\r\nNx *= NormalStrength;\r\nNy *= NormalStrength;\r\n\r\nNormal = normalize(float3(Nx, Ny, Nz));\r\n\r\nWave0 = waveGray[0];\r\nWave1 = waveGray[1];\r\nWave2 = waveGray[2];\r\nWave3 = waveGray[3];\r\nWave4 = waveGray[4];\r\nWave5 = waveGray[5];\r\n\r\nreturn offset;"
   showCode=true
   additionalOutputs [
    {outputName=Normal, outputType=CMOT_Float3}
    {outputName=Wave0, outputType=CMOT_Float1}
    {outputName=Wave1, outputType=CMOT_Float1}
    {outputName=Wave2, outputType=CMOT_Float1}
    {outputName=Wave3, outputType=CMOT_Float1}
    {outputName=Wave4, outputType=CMOT_Float1}
    {outputName=Wave5, outputType=CMOT_Float1}
   ]
   inputs [
    {
     inputName=Position
     input {expression=&self:MaterialExpressionMultiply_26}
    }
    {
     inputName=Time
     input {expression=&self:MaterialExpressionMultiply_27}
    }
    {
     inputName=Steepness
     input {expression=&self:MaterialExpressionReroute_181}
    }
    {
     inputName=Amplitude
     input {expression=&self:MaterialExpressionReroute_182}
    }
    {
     inputName=NormalStrength
     input {expression=&self:MaterialExpressionReroute_196}
    }
   ]
  }
  NamedRerouteDeclaration(None=VertexInterpolator(VS=@n3.Normal).PS) {name=NormalOut, nodeColor=<rgba 0 1 1 1>}
  NamedRerouteDeclaration(None=@n3.return) {name=WPOOffset, nodeColor=<rgba 1 0.6797 0 1>}
  n4 = Power(Base=MaterialFunctionCall(Distance=ScalarParameter() {defaultValue=9333.1758, group=06_Distance, parameterName=StartFadeDistance}) {materialFunction=&Near_Camera_Fade}.Result) {constExponent=8}
  OneMinus(None=@n4)
  n5 = ScalarParameter() {defaultValue=230, group=01_Wave, parameterName=WaveAmplitudeScale}
  n6 = Divide(A=@n5) {constB=-2}
  n7 = NamedRerouteUsage()
  MaterialExpressionTextureCoordinate_3 = TextureCoordinate()
  n8 = Convert(None=@MaterialExpressionTextureCoordinate_3) {
   nodeName=Break Float2
   convertInputs [{
     type=Vector2
     expressionInput {
      expression=&self:MaterialExpressionTextureCoordinate_3
      inputName=None
      mask=0
      maskA=0
      maskB=0
      maskG=0
      maskR=0
      outputIndex=0
     }
    }]
   convertOutputs [{type=Scalar}, {type=Scalar}]
  }
  n9 = VectorParameter() {
   group=WPOEdgeMask
   parameterName=WPOEdgeMask_EdgeMask_InvertControl
   channelNames {a="", b=Power, g=Multiply, r=Offset}
   defaultValue=<rgba 0 4 1 1>
  }
  n10 = MaterialFunctionCall!(UorV=@n8, Invert_EdgeControl=@n9.RGB, NoInvert_EdgeControl=@n9.RGB) {materialFunction=/Game/Effects/Niagara/MasterMaterials/Functions/Common/FXCommon_EdgeMask_Ex}
  n11 = VectorParameter() {
   group=WPOEdgeMask
   parameterName=WPOEdgeMask_EdgeMask_Control
   channelNames {a="", b=Power, g=Multiply, r=Offset}
   defaultValue=<rgba 0 4 1 1>
  }
  n12 = MaterialFunctionCall(UorV=@n8, Invert_EdgeControl=@n9.RGB, NoInvert_EdgeControl=@n11.RGB) {materialFunction=&FXCommon_EdgeMask_Ex}
  n13 = Multiply(A=LinearInterpolate(A=@n6, B=Multiply(A=@n7, B=@n5), Alpha=Multiply(A=@n10.Mask, B=@n12.Mask)), B=VertexNormalWS())
  MaterialExpressionAdd_11 = Add(A=@n13, B=WorldPosition().XYZ)
  n14 = Convert(None=@MaterialExpressionAdd_11) {
   nodeName=Break Float3
   convertInputs [{
     type=Vector3
     expressionInput {
      expression=&self:MaterialExpressionAdd_11
      inputName=None
      mask=0
      maskA=0
      maskB=0
      maskG=0
      maskR=0
      outputIndex=0
     }
    }]
   convertOutputs [{type=Scalar}, {type=Scalar}, {type=Scalar}]
  }
  n15 = Multiply(A=VertexInterpolator(VS=@n3.Wave4).PS, B=ScalarParameter() {defaultValue=110.8251, group=FoamT, parameterName=UVOffset_Foam})
  n16 = Max(A=Subtract(A=Add(A=@n15, B=@n14), B=ScalarParameter() {defaultValue=1622.3865, group=02_SurfaceZ, parameterName=WaterSurfaceHeight_AF})) {constB=0}
  n17 = ScalarParameter() {defaultValue=240.3914, group=WaterColor, parameterName=WaterBrightness}
  n18 = Power(
   Base=OneMinus(None=Divide(A=@n16, B=@n17))
   Exp=ScalarParameter() {defaultValue=0.3704, group=WaterColor, parameterName=WaterColor_G2C_InvtPower}
  )
  n19 = VertexInterpolator(VS=NamedRerouteUsage())
  n20 = ScalarParameter() {defaultValue=30.9927, group=ColorG, parameterName=ColorG_GrayAdjust_Add}
  n21 = Saturate(None=Multiply(A=Add(A=@n19.PS, B=@n20), B=ScalarParameter() {defaultValue=0.0576, group=ColorG, parameterName=ColorG_GrayAdjust_Mul}))
  n22 = ScalarParameter() {defaultValue=2, group=ColorG, parameterName=ColorG_GrayAdjust_Power}
  n23 = LinearInterpolate(
   A=VectorParameter() {group=05_Color, parameterName=AccentColor, defaultValue=<rgba 0.04 0.2779 0.4 1>}.RGB
   B=LinearInterpolate(
    A=VectorParameter() {group=WaterColor, parameterName=WaterColor_GradientA, defaultValue=<rgba 0.0741 0.337 0.6 0>}.RGB
    B=VectorParameter() {group=WaterColor, parameterName=WaterColor_GradientB, defaultValue=<rgba 0 0.7 0.5367 0>}.RGB
    Alpha=OneMinus(None=@n18)
   )
   Alpha=Saturate(None=Power(Base=@n21, Exp=@n22))
  )
  n24 = Desaturation(None=@n23, Fraction=Constant() {r=0.5})
  n25 = VectorParameter() {group=WaveFrontMask, parameterName=WaveDirection, defaultValue=<rgba 1 1 0 1>}
  n26 = Multiply(A=DotProduct(A=Normalize(VectorInput=@n25.RGB), B=NamedRerouteUsage())) {constB=-1}
  n27 = ScalarParameter() {defaultValue=-0.04, group=WaveFrontMask, parameterName=WaveFrontMask_GrayAdjust_Add}
  n28 = Saturate(None=Multiply(A=Add(A=@n26, B=@n27), B=ScalarParameter() {defaultValue=1, group=WaveFrontMask, parameterName=WaveFrontMask_GrayAdjust_Mul}))
  n29 = ScalarParameter() {defaultValue=3, group=WaveFrontMask, parameterName=WaveFrontMask_GrayAdjust_Power}
  n30 = Add(A=@n15, B=ScalarParameter() {defaultValue=-1879.287, group=Spike, parameterName=Spike_GrayAdjust_Add})
  n31 = Saturate(None=Multiply(A=Add(A=@n14, B=@n30), B=ScalarParameter() {defaultValue=0.0005, group=Spike, parameterName=Spike_GrayAdjust_Mul}))
  n32 = ScalarParameter() {defaultValue=1, group=Spike, parameterName=Spike_GrayAdjust_Power}
  n33 = Add(A=Saturate(None=Power(Base=@n28, Exp=@n29)), B=Saturate(None=Power(Base=@n31, Exp=@n32)))
  n34 = VectorParameter() {group=FoamT, parameterName=FoamT_UVControl, channelNames {a=VPan, b=UPan, g=VScale, r=UScale}, defaultValue=<rgba 1.25 1.25 -0.1 -0.1>}
  n35 = MaterialFunctionCall(
   Texture=TextureObject() {samplerType=SAMPLERTYPE_Masks, texture=/FortWater/Textures/T_OceanFoam_New_Mask}
   UVControl_RGB=@n34.RGB
   UVControl_A=@n34.VPan
  ) {materialFunction=&FXCommon_TextureSample}
  n36 = ScalarParameter() {defaultValue=2, group=99_Misc, parameterName=a2}
  n37 = Multiply(
   A=VectorParameter() {group=05_Color, parameterName=BaseColorTint, defaultValue=<rgba 23.4323 7.8261 60 1>}.RGB
   B=Multiply(A=Saturate(None=@n33), B=Power(Base=@n35.R, Exp=@n36))
  )
  Reroute(None=Add(
    A=MaterialFunctionCall!("Xb1,PS4"=@n23, "PC-High&&Epic,XSX,PS5"=@n23, "PC-Perf,Android"=@n24, Switch=@n24, "PC-Low(NotPerf)&&Medium"=@n23) {materialFunction=/Game/Effects/Niagara/MasterMaterials/Functions/Common/PlatformSwitch_3D}.Result
    B=@n37
   ))
  n38 = ReflectionVectorWS(CustomWorldNormal=NamedRerouteUsage())
  n39 = ScalarParameter() {parameterName=Roughness}
  n40 = Multiply(
   A=Multiply(A=SkyLightEnvMapSample(Direction=@n38, Roughness=@n39), B=ScalarParameter() {defaultValue=0.6, parameterName=ReflectStr})
   B=Fresnel() {baseReflectFraction=0, exponent=3}
  )
  n41 = Constant()
  Reroute(None=Add(
    A=MaterialFunctionCall(Emissive=@n37) {materialFunction=&MF_FX_EmissiveAdjustments}.Result
    B=MaterialFunctionCall("Xb1,PS4"=@n40, "PC-High&&Epic,XSX,PS5"=@n40, "PC-Perf,Android"=@n41, Switch=@n41, "PC-Low(NotPerf)&&Medium"=@n40) {materialFunction=&PlatformSwitch_3D}.Result
   ))
  n42 = MaterialFunctionCall() {materialFunction=&CameraDirectionVector}
  n43 = Constant3Vector() {constant=<rgba 1 1 0 0>}
  n44 = DotProduct(A=@n42.Result, B=Transform(None=Normalize(VectorInput=@n43)))
  n45 = Multiply(
   A=Subtract(A=Saturate(None=@n44), B=ScalarParameter() {defaultValue=0.5, group=AngleLimit, parameterName=XYAngleLimit_Offset})
   B=ScalarParameter() {defaultValue=4, group=AngleLimit, parameterName=XYAngleLimit_Mul}
  )
  n46 = DotProduct(A=@n42.Result, B=Constant3Vector() {constant=<rgba 0 0 1 0>})
  n47 = Multiply(
   A=Subtract(A=Abs(None=@n46), B=ScalarParameter() {defaultValue=0.224, group=AngleLimit, parameterName=ZAngleLimit_Offset})
   B=ScalarParameter() {defaultValue=10, group=AngleLimit, parameterName=ZAngleLimit_Mul}
  )
  n48 = Saturate(None=LinearInterpolate(A=Saturate(None=@n45), Alpha=Saturate(None=@n47)) {constB=0})
  n49 = Max(A=Subtract(A=@n14, B=ScalarParameter() {defaultValue=1380, group=02_SurfaceZ, parameterName=WaterSurfaceHeight})) {constB=0}
  n50 = ScalarParameter() {defaultValue=1083.0216, group=02_SurfaceZ, parameterName=WaterSurfaceFade}
  n51 = OneMinus(None=Power(Base=Divide(A=@n49, B=@n50), Exp=ScalarParameter() {defaultValue=1.2, group=02_SurfaceZ, parameterName=SurfaceCurveInner}))
  n52 = ScalarParameter() {defaultValue=4, group=02_SurfaceZ, parameterName=SurfaceCurveOuter}
  n53 = Multiply(
   A=OneMinus(None=Power(Base=@n51, Exp=@n52))
   B=DepthFade(FadeDistance=ScalarParameter() {defaultValue=100, group=03_Opacity, parameterName=DepthFadeDistance})
  )
  n54 = LinearInterpolate(
   A=ScalarParameter() {defaultValue=0.5, group=03_Opacity, parameterName=OpacityScale}
   B=ScalarParameter() {defaultValue=0.25, group=03_Opacity, parameterName=OpacityScale2}
   Alpha=@n4
  )
  n55 = Power(Base=MaterialFunctionCall(Distance=ScalarParameter() {defaultValue=1800, group=06_Distance, parameterName=NCF_Distance}) {materialFunction=&Near_Camera_Fade}.Result) {constExponent=1.5}
  n56 = Multiply(A=Multiply(A=OneMinus(None=@n48), B=Multiply(A=@n53, B=@n54)), B=@n55)
  Reroute(None=@n56)
  Reroute(None=Transform(None=NamedRerouteUsage()) {transformSourceType=TRANSFORMSOURCE_Local})
  MaterialExpressionRotator_1 = Rotator(Coordinate=TextureCoordinate(), Time=MaterialFunctionCall() {materialFunction=&Pi}.Result)
  n57 = Convert(None=@MaterialExpressionRotator_1) {
   nodeName=Break Float2
   convertInputs [{
     type=Vector2
     expressionInput {
      expression=&self:MaterialExpressionRotator_1
      inputName=None
      mask=0
      maskA=0
      maskB=0
      maskG=0
      maskR=0
      outputIndex=0
     }
    }]
   convertOutputs [{type=Scalar}, {type=Scalar}]
  }
  n58 = VectorParameter() {
   group=StrengthMask
   parameterName=StrengthMask_EdgeMask_Control
   channelNames {a="", b=Power, g=Multiply, r=Offset}
   defaultValue=<rgba 10 1 1 1>
  }
  n59 = VectorParameter() {
   group=StrengthMask
   parameterName=StrengthMask_EdgeMask_InvertControl
   channelNames {a="", b=Power, g=Multiply, r=Offset}
   defaultValue=<rgba 0 4.5 0.5 1>
  }
  Reroute(None=LinearInterpolate(
    A=@n6
    B=Multiply(A=@n13, B=@n55)
    Alpha=MaterialFunctionCall(UorV=@n57, Invert_EdgeControl=@n59.RGB, NoInvert_EdgeControl=@n58.RGB) {materialFunction=&FXCommon_EdgeMask_Ex}.Mask
   ))
  Reroute(None=LinearInterpolate(B=ScalarParameter() {defaultValue=1.1, group=04_PBR, parameterName=RefractionIOR}, Alpha=@n56) {constA=1})
 ]
}
