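# Unlit translucent Niagara sprite material for Zero Point water drops, shading each sprite as a refractive sphere.
# Noise-distorted UVs feed custom HLSL sphere normals and a SphereMask opacity scaled by particle alpha.
# Those normals drive two GGX highlights, Fresnel cubemap reflection and refracted SceneColor, summed and tinted by time of day.
# Dynamic parameters control refraction strength and noise offset; low-end platforms use flat color and reduced opacity instead of refraction.
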Material /Game/Environments/Hera/FX/ZeroPoint/WaterDisturbance/M_Env_ZeroPoint_Drop
graph {
 properties {
  bAutomaticallySetUsageInEditor=false
  bUsedWithNiagaraSprites=true
  blendMode=BLEND_Translucent
  shadingModel=MSM_Unlit
  translucencyPass=MTP_BeforeDOF
  usedShadingModels=Unlit
 }
 nodes [
  n1 = ScalarParameter() {defaultValue=0.05, parameterName=Rou}
  n2 = TextureObjectParameter() {
   group=Noise
   parameterName=Noise_Texture
   samplerType=SAMPLERTYPE_Normal
   sortPriority=0
   texture=/Game/Effects/Fort_Effects/Emotes/BirdNestBlue/T_LowResSimplePerlin
  }
  MaterialExpressionDynamicParameter_0 = DynamicParameter() {defaultValue=<rgba 0.0976 1 1 1>, paramNames [Ref,RDm,Param3,Param4]}
  n3 = VectorParameter() {group=Noise, parameterName=Noise_UVControl, channelNames {a=VPan, b=UPan, g=VScale, r=UScale}, defaultValue=<rgba 0.18 0.18 0 -1>}
  n4 = VectorParameter() {
   group=Noise
   parameterName=Noise_NoiseControl
   sortPriority=60
   channelNames {a=VStrength, b=UStrength, g=VOffset, r=UOffset}
   defaultValue=<rgba 0 0 0.1 0.1>
  }
  MaterialExpressionAdd_3 = Add(
   A=MaterialFunctionCall!(
    NoiseValue=MaterialFunctionCall(Texture=@n2, AdditionalUVOffset=@MaterialExpressionDynamicParameter_0.RDm, UVControl_RGB=@n3.RGB, UVControl_A=@n3.RGB) {materialFunction=&FXCommon_TextureSample}.RG
    Control_RGB=@n4.RGB
    Control_A=@n4.RGB
   ) {materialFunction=/Game/Effects/Niagara/MasterMaterials/Functions/Common/FXCommon_NoiseControl}.NoiseValue
   B=TextureCoordinate()
  )
  MaterialExpressionVectorParameter_0 = VectorParameter() {parameterName=Sphere, defaultValue=<rgba 0.5 0.5 0.4 1>}
  MaterialExpressionAppendVector_0 = AppendVector(A=@MaterialExpressionVectorParameter_0.R, B=@MaterialExpressionVectorParameter_0.R)
  n5 = Custom!(UV=@MaterialExpressionAdd_3, Center=@MaterialExpressionAppendVector_0, Radius=@MaterialExpressionVectorParameter_0.B) {
   code="float2 p = (UV - Center) / max(Radius, 1e-5);\r\n\r\nfloat r2 = dot(p, p);\r\n\r\nfloat z = sqrt(saturate(1.0 - r2));\r\n\r\nfloat3 N = normalize(float3(p.x, p.y, z));\r\n\r\nfloat inside = step(r2, 1.0);\r\nN = lerp(float3(0.0, 0.0, 1.0), N, inside);\r\n\r\nreturn N;"
   showCode=true
   inputs [
    {inputName=UV, input {expression=&self:MaterialExpressionAdd_3}}
    {
     inputName=Center
     input {expression=&self:MaterialExpressionAppendVector_0}
    }
    {
     inputName=Radius
     input {
      expression=&self:MaterialExpressionVectorParameter_0
      mask=1
      maskB=1
      outputIndex=3
     }
    }
   ]
  }
  n6 = Transform(None=@n5)
  n7 = Multiply(A=SkyAtmosphereLightDirection()) {constB=-1}
  n8 = Multiply(
   A=Add(
    A=MaterialFunctionCall!(LightDirctionIn, CameraVectorIn, Roughness=@n1, NormalIn=@n6) {materialFunction=/Game/Effects/Niagara/MasterMaterials/Functions/Common/FXCommon_GGX}.Result
    B=MaterialFunctionCall(LightDirctionIn=@n7, Roughness=@n1, NormalIn=@n6) {materialFunction=&FXCommon_GGX}.Result
   )
   B=VectorParameter() {parameterName=GGX, defaultValue=<rgba 100 100 100 1>}.RGB
  )
  n9 = ReflectionVectorWS(CustomWorldNormal=@n6)
  n10 = VectorParameter() {parameterName=Ref, defaultValue=<rgba 2.4869 1.6017 3 1>}
  n11 = VectorParameter() {parameterName=Fres, defaultValue=<rgba 3 0.1 0 1>}
  n12 = Fresnel(ExponentIn=@n11.R, BaseReflectFractionIn=@n11.R, Normal=@n6) {baseReflectFraction=0.5}
  MaterialExpressionScreenPosition_0 = ScreenPosition()
  MaterialExpressionTransform_0 = Transform(None=@n5) {transformType=TRANSFORM_View}
  MaterialExpressionTransform_1 = Transform(None=CameraVectorWS()) {transformSourceType=TRANSFORMSOURCE_World, transformType=TRANSFORM_View}
  MaterialExpressionScalarParameter_0 = ScalarParameter() {defaultValue=2, parameterName=IOR}
  MaterialExpressionDotProduct_0 = DotProduct(A=@n5, B=Constant3Vector() {constant=<rgba 0 0 1 0>})
  n13 = Custom!(
   ScreenUV=@MaterialExpressionScreenPosition_0.ViewportUV
   NormalVS=@MaterialExpressionTransform_0
   CameraVectorVS=@MaterialExpressionTransform_1
   IOR=@MaterialExpressionScalarParameter_0
   Thickness=@MaterialExpressionDotProduct_0
   Strength=@MaterialExpressionDynamicParameter_0.Ref
  ) {
   code="float3 N = normalize(NormalVS);\r\nfloat3 V = normalize(CameraVectorVS);\r\nN = dot(N, V) < 0.0 ? -N : N;\r\n\r\n\r\nfloat eta = rcp(max(IOR, 1.0001));\r\n\r\nfloat3 I = -V;\r\nfloat3 R = refract(I, N, eta);\r\n\r\n\r\nR = dot(R, R) > 1e-6 ? R : I;\r\n\r\n\r\nfloat2 offset = R.xy - I.xy;\r\n\r\noffset.y *= -1.0;\r\n\r\nreturn saturate(ScreenUV + offset * Thickness * Strength);"
   outputType=CMOT_Float2
   showCode=true
   inputs [
    {
     inputName=ScreenUV
     input {expression=&self:MaterialExpressionScreenPosition_0}
    }
    {
     inputName=NormalVS
     input {expression=&self:MaterialExpressionTransform_0}
    }
    {
     inputName=CameraVectorVS
     input {expression=&self:MaterialExpressionTransform_1}
    }
    {
     inputName=IOR
     input {expression=&self:MaterialExpressionScalarParameter_0}
    }
    {
     inputName=Thickness
     input {expression=&self:MaterialExpressionDotProduct_0}
    }
    {
     inputName=Strength
     input {expression=&self:MaterialExpressionDynamicParameter_0, mask=1, maskR=1}
    }
   ]
  }
  n14 = Multiply(A=SceneColor!(UVs=@n13).RGB, B=VectorParameter() {parameterName=WaterC, defaultValue=<rgba 0.207 0.3861 0.4035 1>}.RGB)
  n15 = VectorParameter() {parameterName=LowEndColor, defaultValue=<rgba 0.1883 0.27 0.4383 1>}
  n16 = Add(
   A=Multiply(
    A=TextureSampleParameterCube!(UVs=@n9, Apply View MipBias) {parameterName=Ref, texture=/Game/TimeOfDay/TODM_Sequence/Textures/SkyDome/Asteria/Asteria_Day}.RGB
    B=Multiply(A=@n10.RGB, B=@n12)
   )
   B=MaterialFunctionCall!("Xb1,PS4"=@n14, "PC-High&&Epic,XSX,PS5"=@n14, "PC-Perf,Android"=@n15.RGB, Switch=@n15.RGB, "PC-Low(NotPerf)&&Medium"=@n14) {materialFunction=/Game/Effects/Niagara/MasterMaterials/Functions/Common/PlatformSwitch_3D}.Result
  )
  n17 = VectorParameter() {parameterName="Color_Morning (V3)", defaultValue=<rgba 0.9353 0.7898 1 1>}
  n18 = VectorParameter() {parameterName="Color_Day (V3)", defaultValue=<rgba 1 1 1 0>}
  n19 = VectorParameter() {parameterName="Color_Night (V3)", defaultValue=<rgba 0.5425 0.9018 1 1>}
  MaterialFunctionCall(Emissive=Multiply(
    A=Add(A=@n8, B=@n16)
    B=MaterialFunctionCall(Color_Morning=@n17.RGB, Color_Day=@n18.RGB, Color_Evening=@n17.RGB, Color_Night=@n19.RGB) {materialFunction=&Color_Time_of_Day}.Result
   )) {materialFunction=&MF_FX_EmissiveAdjustments}
  n20 = SphereMask(
   A=@MaterialExpressionAppendVector_0
   B=@MaterialExpressionAdd_3
   Radius=@MaterialExpressionVectorParameter_0.B
   Hardness=ScalarParameter() {defaultValue=1, parameterName=Hdar}
  ) {hardnessPercent=5}
  n21 = Multiply(A=@n20, B=ScalarParameter() {defaultValue=0.25, parameterName=LowEndOpacity})
  Multiply(
   A=MaterialFunctionCall("Xb1,PS4"=@n20, "PC-High&&Epic,XSX,PS5"=@n20, "PC-Perf,Android"=@n21, Switch=@n21, "PC-Low(NotPerf)&&Medium"=@n20) {materialFunction=&PlatformSwitch_3D}.Result
   B=ParticleColor().A
  )
 ]
}
