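# Translucent unlit water splash material for Niagara sprites in the Zero Point water disturbance effects.
# Particle color drives emissive through an adjustment function, and particle alpha scales opacity.
# Opacity multiplies a noise-distorted splash texture by a contrast-adjusted dissolve mask that the Erode dynamic parameter offsets; both noises can pan, though default speeds are zero.
# A quality switch applies depth fade on Default, High and Epic, skipping it on Low and Medium.
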Material /Game/Environments/Hera/FX/ZeroPoint/WaterDisturbance/M_Env_ZeroPoint_WaterSplash_Tranlucent
graph {
 properties {
  bAutomaticallySetUsageInEditor=false
  bUsedWithNiagaraSprites=true
  blendMode=BLEND_Translucent
  previewMesh=/Game/Environments/Hera/FX/Meshes/SM_ZeroPoint_RiverFlatMesh_01
  shadingModel=MSM_Unlit
  usedShadingModels=Unlit
 }
 nodes [
  n1 = ParticleColor()
  MaterialFunctionCall(Emissive=@n1.RGB) {materialFunction=&MF_FX_EmissiveAdjustments}
  n2 = TextureCoordinate()
  n3 = VectorParameter() {group=10.Noise_ForTexture_A, parameterName=Noise_ForTexture_A_ScaleAndSpeed, defaultValue=<rgba 1 1 0 0.7>}
  n4 = AppendVector(A=@n3.R, B=@n3.R)
  n5 = AppendVector(A=@n3.B, B=@n3.B)
  n6 = Time() {bOverride_Period=true, period=333}
  n7 = TextureSample(UVs=Add(A=Multiply(A=@n2, B=@n4), B=Multiply(A=@n5, B=@n6))) {
   samplerType=SAMPLERTYPE_Normal
   texture=/Game/Athena/Artemis/Environments/BuildingActors/WaterFall/Materials/Texture/T_Lowres_Simple_Perlin_N
  }
  n8 = VectorParameter() {group=10.Noise_ForTexture_A, parameterName=Noise_ForTexture_A_Strength, defaultValue=<rgba 0.025 0 0 0>}
  n9 = Add(A=TextureCoordinate(), B=Multiply(A=@n7.G, B=@n8.R))
  n10 = VectorParameter() {group=10.Noise_ForTexture_A, parameterName=Erode_ForTexture_A_ScaleAndSpeed, defaultValue=<rgba 1 1 0 0.5>}
  n11 = Add(
   A=Multiply(A=TextureCoordinate() {uTiling=2, vTiling=2}, B=AppendVector(A=@n10.R, B=@n10.R))
   B=Multiply(A=AppendVector(A=@n10.B, B=@n10.B), B=@n6)
  )
  n12 = Saturate(None=Add(
    A=DynamicParameter() {paramNames [Erode,Param2,Param3,Param4]}.Erode
    B=TextureSample(UVs=@n11) {texture=/Game/CharClasses/Outlander/Gadgets/AntiMaterialCharge/Materials/T_AntiMaterialCharge_Dissolve_01}.B
   ))
  n13 = ScalarParameter() {defaultValue=1, parameterName=C}
  n14 = Multiply(
   A=@n1.A
   B=Multiply(
    A=TextureSample(UVs=@n9) {samplerType=SAMPLERTYPE_LinearColor, texture=/Game/Effects/Fort_Effects/Textures/Water/T_waterSplash}.G
    B=MaterialFunctionCall(In=@n12, Contrast=@n13) {materialFunction=&CheapContrast}.Result
   )
  )
  n15 = DepthFade(Opacity=@n14, FadeDistance=ScalarParameter() {parameterName=FadeDistance})
  QualitySwitch(Default=@n15, Low=@n14, High=@n15, Medium=@n14, Epic=@n15)
 ]
}
