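# Masked, lit material for Niagara mesh particles that draws water wave disturbance effects in the Zero Point environment.
# A water displacement texture, tiled by dynamic parameters and panned with per-particle random speed, drives tip-fade and dissolve thresholds and blends two cyan tints through emissive adjustments.
# Edge masks plus foam noise break up the silhouette.
# Bayer dithering softens lower mask values, which are combined with particle alpha and a bounding-box gradient.
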Material /Game/Environments/Hera/FX/ZeroPoint/WaterDisturbance/M_Env_ZeroPoint_WaterWaves_Masked
graph {
 properties {bAutomaticallySetUsageInEditor=false, bUsedWithNiagaraMeshParticles=true, blendMode=BLEND_Masked, usedShadingModels=Default Lit}
 nodes [
  n1 = MaterialFunctionCall() {materialFunction=&BoundingBoxBased_0-1_UVW}
  n2 = OneMinus(None=@n1.R)
  n3 = ScalarParameter() {defaultValue=0.4, parameterName=Param_8}
  n4 = DynamicParameter() {defaultValue=<rgba 1 -0.656 1 1>, paramNames [TipFade,Dissolve,Tile_X,Tile_Y]}
  n5 = TextureCoordinate()
  n6 = DynamicParameter() {defaultValue=<rgba 1 -0.656 1 1>, paramNames [TipFade,Dissolve,Tile_X,Tile_Y]}
  n7 = AppendVector(A=@n6.Tile_X, B=@n6.Tile_X)
  n8 = Time()
  n9 = ScalarParameter() {defaultValue=1, parameterName=Param_4}
  n10 = ScalarParameter() {parameterName=Param_2}
  n11 = Clamp(None=ParticleRandom()) {maxDefault=-0.6, minDefault=-0.4}
  n12 = TextureSample(UVs=Panner(Coordinate=Multiply(A=@n5, B=@n7), Time=Multiply(A=@n8, B=@n9), Speed=AppendVector(A=@n10, B=@n11))) {texture=/Game/Packages/DS_Fortnite_TRV_Terrain/Textures/T_TRV_HF_WaterDisp}
  n13 = SmoothStep(Min=@n4.TipFade, Value=ComponentMask(None=TextureCoordinate()) {g=true}) {constMin=1}
  n14 = VectorParameter() {parameterName=MaskUVScale, defaultValue=<rgba 1 1.2251 0 1>}
  MaterialExpressionMultiply_7 = Multiply(A=TextureCoordinate(), B=AppendVector(A=@n14.R, B=@n14.R))
  n15 = Convert(None=@MaterialExpressionMultiply_7) {
   nodeName=Break Float2
   convertInputs [{
     type=Vector2
     expressionInput {
      expression=&self:MaterialExpressionMultiply_7
      inputName=None
      mask=0
      maskA=0
      maskB=0
      maskG=0
      maskR=0
      outputIndex=0
     }
    }]
   convertOutputs [{type=Scalar}, {type=Scalar}]
  }
  n16 = VectorParameter() {group=Edge, parameterName=Edge_EdgeMask_Control, channelNames {a="", b=Power, g=Multiply, r=Offset}, defaultValue=<rgba -0.1088 1 0.8 4>}
  n17 = Multiply(
   A=Min(
    A=MaterialFunctionCall!(UorV=@n15, EdgeControl=@n16.RGB) {materialFunction=/Game/Effects/Niagara/MasterMaterials/Functions/Common/FXCommon_EdgeMask_Simple}.Mask
    B=MaterialFunctionCall(UorV=@n15, EdgeControl=@n16.RGB) {materialFunction=&FXCommon_EdgeMask_Simple}.Mask
   )
   B=@n16.A
  )
  n18 = TextureObjectParameter() {
   group=AdditionNoise
   parameterName=AdditionNoise_Texture
   samplerType=SAMPLERTYPE_Masks
   sortPriority=0
   texture=/FortWater/Textures/T_OceanFoam_New_Mask
  }
  n19 = VectorParameter() {group=AdditionNoise, parameterName=EdgeFade_UVControl, channelNames {a=VPan, b=UPan, g=VScale, r=UScale}, defaultValue=<rgba 0.3 0.3 0 0>}
  n20 = Subtract(
   A=MaterialFunctionCall(Texture=@n18, UVControl_RGB=@n19.RGB, UVControl_A=@n19.RGB) {materialFunction=&FXCommon_TextureSample}.R
   B=ScalarParameter() {defaultValue=0.2352, parameterName=DissovleOffset}
  )
  n21 = Multiply(A=SmoothStep(Min=@n4.Dissolve, Value=Add(A=@n12.R, B=@n13)), B=Saturate(None=Add(A=@n17, B=@n20)))
  n22 = MaterialFunctionCall!(In=@n21, Iterations=ScalarParameter() {defaultValue=1}, Scale=ScalarParameter() {defaultValue=1, parameterName=Param_1}) {materialFunction=/Game/Packages/Fortress_SharedMaterials/Base_Material_Functions/MF_DitherBayerNoTex}
  n23 = If(A=@n21, "A > B"=Constant() {r=1}, "A == B"=@n22.Result, "A < B"=@n22.Result) {constB=0.8}
  Multiply(A=Multiply(A=Power(Base=@n2, Exp=@n3) {constExponent=0.2}, B=Multiply(A=@n1.B) {constB=0.9}), B=@n23)
  ParticleRandom()
  n24 = DynamicParameter() {defaultValue=<rgba 1 -0.656 1 1>, paramNames [TipFade,Dissolve,Tile_X,Tile_Y]}
  n25 = ParticleDirection()
  PreviousFrameSwitch(Current Frame=VertexNormalWS(), Previous Frame=Subtract(A=VertexNormalWS(), B=Multiply(A=@n24.TipFade, B=@n25)))
  Reroute(None=Multiply(A=ParticleColor().A, B=@n23))
  n26 = VectorParameter() {parameterName=Param_1, defaultValue=<rgba 0.1757 0.2921 0.3256 1>}
  n27 = VectorParameter() {defaultValue=<rgba 0.401 0.9148 1 1>}
  Reroute(None=MaterialFunctionCall(Emissive=LinearInterpolate(A=@n26.RGB, B=@n27.RGB, Alpha=@n12.RGB)) {materialFunction=&MF_FX_EmissiveAdjustments}.Result)
  ScalarParameter() {defaultValue=1, parameterName=Param_6}
  ScalarParameter() {defaultValue=3, parameterName=Param_7}
  ScalarParameter() {defaultValue=0.594, parameterName=Param_10}
  ScalarParameter() {defaultValue=500, parameterName=Param_5}
 ]
}
