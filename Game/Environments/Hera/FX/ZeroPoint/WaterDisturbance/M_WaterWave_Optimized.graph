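# Translucent water wave material for Niagara mesh particles, likely for Zero Point water disturbance.
# Custom HLSL builds scrolling wave height and analytic normals without textures, with hard-coded shaping that ignores its Falloff, NormalStrength, Roundness and RowOverlap inputs.
# Edge-masked height drives time-of-day-tinted color, depth-faded opacity and foam-noise-modulated vertex-normal displacement, steered by dynamic parameters.
# Fresnel reflection samples the skylight, but a platform switch substitutes flat color on PC-Perf, Android and Switch.
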Material /Game/Environments/Hera/FX/ZeroPoint/WaterDisturbance/M_WaterWave_Optimized
graph {
 expressions {
  MaterialExpressionAdd_4 {
   class=MaterialExpressionAdd
   inputs [
    A {from=MaterialExpressionTime_1}
    B {from=MaterialExpressionDynamicParameter_3, output=Param4}
   ]
  }
  MaterialExpressionAdd_5 {
   class=MaterialExpressionAdd
   inputs [
    A {from=MaterialExpressionMultiply_30}
    B {from=MaterialExpressionScalarParameter_1}
   ]
  }
  MaterialExpressionAdd_6 {
   class=MaterialExpressionAdd
   inputs [
    A {from=MaterialExpressionReroute_31}
    B {from=MaterialExpressionScalarParameter_5}
   ]
  }
  MaterialExpressionAdd_7 {
   class=MaterialExpressionAdd
   inputs [
    A {from=MaterialExpressionTextureCoordinate_2}
    B {from=MaterialExpressionConvert_2}
   ]
  }
  MaterialExpressionAppendVector_1 {
   class=MaterialExpressionAppendVector
   inputs [
    A {from=MaterialExpressionReroute_17, output=""}
    B {from=MaterialExpressionReroute_18, output=""}
   ]
  }
  MaterialExpressionComponentMask_2 {class=MaterialExpressionComponentMask, inputs [None {from=MaterialExpressionReroute_16}], properties {b=true, g=true, r=true}}
  MaterialExpressionComponentMask_3 {class=MaterialExpressionComponentMask, inputs [None {from=MaterialExpressionCustom_1}], properties {a=true}}
  MaterialExpressionConstant_10 {class=MaterialExpressionConstant, properties {r=20}}
  MaterialExpressionConstant_11 {class=MaterialExpressionConstant, properties {r=0.1}}
  MaterialExpressionConstant_12 {class=MaterialExpressionConstant}
  MaterialExpressionConstant_13 {class=MaterialExpressionConstant, properties {r=2}}
  MaterialExpressionConstant_14 {class=MaterialExpressionConstant, properties {r=1.3}}
  MaterialExpressionConstant_15 {class=MaterialExpressionConstant, properties {r=1.5}}
  MaterialExpressionConstant_8 {class=MaterialExpressionConstant}
  MaterialExpressionConstant_9 {class=MaterialExpressionConstant, properties {r=1}}
  MaterialExpressionConvert_0 {
   class=MaterialExpressionConvert
   inputs [None {from=MaterialExpressionTextureCoordinate_0}]
   properties {
    nodeName=Break Float2
    convertInputs [{
      type=Vector2
      expressionInput {
       expression=&self:MaterialExpressionTextureCoordinate_0
       inputName=None
       mask=0
       maskA=0
       maskB=0
       maskG=0
       maskR=0
       outputIndex=0
      }
     }]
    convertOutputs [{type=Scalar}, {type=Scalar}]
   }
  }
  MaterialExpressionConvert_2 {
   class=MaterialExpressionConvert
   inputs [
    None {from=""}
    None {from=MaterialExpressionMultiply_29}
   ]
   properties {
    nodeName=Make Vector2
    convertInputs [
     {type=Scalar, expressionInput {expression=null, inputName=None, mask=0, maskA=0, maskB=0, maskG=0, maskR=0, outputIndex=0}}
     {
      type=Scalar
      expressionInput {
       expression=&self:MaterialExpressionMultiply_29
       inputName=None
       mask=0
       maskA=0
       maskB=0
       maskG=0
       maskR=0
       outputIndex=0
      }
     }
    ]
    convertOutputs [{type=Vector2}]
   }
  }
  MaterialExpressionConvert_3 {
   class=MaterialExpressionConvert
   inputs [None {from=MaterialExpressionTextureCoordinate_3}]
   properties {
    nodeName=Break Float2
    convertInputs [{
      type=Vector2
      expressionInput {
       expression=&self:MaterialExpressionTextureCoordinate_3
       inputName=None
       mask=0
       maskA=0
       maskB=0
       maskG=0
       maskR=0
       outputIndex=0
      }
     }]
    convertOutputs [{type=Scalar}, {type=Scalar}]
   }
  }
  MaterialExpressionCustom_1 {
   class=MaterialExpressionCustom
   inputs [
    UV {from=MaterialExpressionAdd_7, output=""}
    Tiling {from=MaterialExpressionAppendVector_1, output=""}
    Falloff {from=MaterialExpressionReroute_19, output=""}
    NormalStrength {from=MaterialExpressionReroute_20, output=""}
    Roundness {from=MaterialExpressionConstant_15, output=""}
    CenterOffsetV {from=MaterialExpressionDynamicParameter_3, output=Param2}
    RowOverlap {from=MaterialExpressionConstant_14, output=""}
   ]
   properties {
    code="float2 uv = UV * Tiling;\r\n\r\nfloat cx = frac(uv.x) - 0.5;\r\nfloat qx = abs(cx + cx);\r\n\r\nfloat peakV = clamp(0.5 + CenterOffsetV, 0.001, 0.999);\r\nfloat invBelow = 0.76923077 * rcp(peakV);\r\nfloat invAbove = 0.76923077 * rcp(1.0 - peakV);\r\n\r\nfloat y = uv.y - peakV;\r\nfloat dy = y - floor(y + 0.5);\r\n\r\nbool below = dy < 0.0;\r\n\r\nfloat inv = below ? invBelow : invAbove;\r\nfloat sgn = below ? -1.0 : 1.0;\r\nfloat qy = saturate(abs(dy) * inv);\r\n\r\nfloat qd = saturate((1.0 + dy) * invAbove);\r\nbool td = qd < qy;\r\nqy  = td ? qd : qy;\r\nsgn = td ? 1.0 : sgn;\r\ninv = td ? invAbove : inv;\r\n\r\nfloat qu = saturate((1.0 - dy) * invBelow);\r\nbool tu = qu < qy;\r\nqy  = tu ? qu : qy;\r\nsgn = tu ? -1.0 : sgn;\r\ninv = tu ? invBelow : inv;\r\n\r\nfloat sx = sqrt(qx);\r\nfloat sy = sqrt(qy);\r\n\r\nfloat A = 1.0 - qx * sx;   // qx^1.5\r\nfloat B = 1.0 - qy * sy;   // qy^1.5\r\n\r\nfloat h = saturate(A * B);\r\n\r\nfloat dhdu = (-3.0 * sign(cx) * sx) * B;\r\nfloat dhdv = A * (-1.5 * sy * sgn * inv);\r\n\r\nfloat3 N = normalize(float3(-dhdu, -dhdv, 1.0));\r\n\r\nreturn float4(N, h);"
    outputType=CMOT_Float4
    showCode=true
    inputs [
     {
      inputName=UV
      input {
       expression=&self:MaterialExpressionAdd_7
       inputName=None
       mask=0
       maskA=0
       maskB=0
       maskG=0
       maskR=0
       outputIndex=0
      }
     }
     {
      inputName=Tiling
      input {
       expression=&self:MaterialExpressionAppendVector_1
       inputName=None
       mask=0
       maskA=0
       maskB=0
       maskG=0
       maskR=0
       outputIndex=0
      }
     }
     {
      inputName=Falloff
      input {
       expression=&self:MaterialExpressionReroute_19
       inputName=None
       mask=0
       maskA=0
       maskB=0
       maskG=0
       maskR=0
       outputIndex=0
      }
     }
     {
      inputName=NormalStrength
      input {
       expression=&self:MaterialExpressionReroute_20
       inputName=None
       mask=0
       maskA=0
       maskB=0
       maskG=0
       maskR=0
       outputIndex=0
      }
     }
     {
      inputName=Roundness
      input {
       expression=&self:MaterialExpressionConstant_15
       inputName=None
       mask=0
       maskA=0
       maskB=0
       maskG=0
       maskR=0
       outputIndex=0
      }
     }
     {
      inputName=CenterOffsetV
      input {
       expression=&self:MaterialExpressionDynamicParameter_3
       inputName=None
       mask=1
       maskA=0
       maskB=0
       maskG=1
       maskR=0
       outputIndex=1
      }
     }
     {
      inputName=RowOverlap
      input {
       expression=&self:MaterialExpressionConstant_14
       inputName=None
       mask=0
       maskA=0
       maskB=0
       maskG=0
       maskR=0
       outputIndex=0
      }
     }
    ]
   }
  }
  MaterialExpressionDepthFade_1 {
   class=MaterialExpressionDepthFade
   inputs [
    Opacity {from="", output=""}
    FadeDistance {from=MaterialExpressionReroute_26, output=""}
   ]
   properties {}
  }
  MaterialExpressionDynamicParameter_2 {class=MaterialExpressionDynamicParameter, inputs [], properties {defaultValue=<rgba 1 0 1 1>}}
  MaterialExpressionDynamicParameter_3 {class=MaterialExpressionDynamicParameter, inputs [], properties {defaultValue=<rgba 1 0 1 1>}}
  MaterialExpressionFresnel_1 {
   class=MaterialExpressionFresnel
   inputs [
    ExponentIn {from=MaterialExpressionConstant_10, output=""}
    BaseReflectFractionIn {from=MaterialExpressionConstant_11, output=""}
    Normal {from=MaterialExpressionReroute_27, output=""}
   ]
   properties {}
  }
  MaterialExpressionLinearInterpolate_0 {
   class=MaterialExpressionLinearInterpolate
   inputs [
    A {from=MaterialExpressionVectorParameter_13, output=RGB}
    B {from=MaterialExpressionReroute_24}
    Alpha {from=MaterialExpressionSaturate_7}
   ]
  }
  MaterialExpressionMaterialFunctionCall_0 {
   class=MaterialExpressionMaterialFunctionCall
   inputs [
    UorV {from=MaterialExpressionConvert_0}
    EdgeControl {from=MaterialExpressionVectorParameter_8, output=RGB}
   ]
   properties {materialFunction=/Game/Effects/Niagara/MasterMaterials/Functions/Common/FXCommon_EdgeMask_Simple}
  }
  MaterialExpressionMaterialFunctionCall_10 {
   class=MaterialExpressionMaterialFunctionCall
   inputs [
    Texture {from=MaterialExpressionTextureObjectParameter_1}
    AdditionalUVOffset {from=MaterialExpressionMultiply_19}
    UVControl_RGB {from=MaterialExpressionReroute_32}
    UVControl_A {from=MaterialExpressionReroute_33}
    "[Override]TimeInput" {from=""}
    "[Override]UVInput" {from=""}
   ]
   properties {materialFunction=&FXCommon_TextureSample}
  }
  MaterialExpressionMaterialFunctionCall_11 {
   class=MaterialExpressionMaterialFunctionCall
   inputs [
    Color_Morning {from=MaterialExpressionReroute_34}
    Color_Day {from=MaterialExpressionReroute_35}
    Color_Evening {from=MaterialExpressionReroute_36}
    Color_Night {from=MaterialExpressionReroute_14}
   ]
   properties {materialFunction=&Color_Time_of_Day}
  }
  MaterialExpressionMaterialFunctionCall_13 {
   class=MaterialExpressionMaterialFunctionCall
   inputs [Emissive {from=MaterialExpressionMultiply_31}]
   properties {materialFunction=&MF_FX_EmissiveAdjustments}
  }
  MaterialExpressionMaterialFunctionCall_2 {
   class=MaterialExpressionMaterialFunctionCall
   inputs [
    UorV {from=MaterialExpressionConvert_0}
    Invert_EdgeControl {from=MaterialExpressionVectorParameter_1, output=RGB}
    NoInvert_EdgeControl {from=MaterialExpressionVectorParameter_2, output=RGB}
   ]
   properties {materialFunction=/Game/Effects/Niagara/MasterMaterials/Functions/Common/FXCommon_EdgeMask_Ex}
  }
  MaterialExpressionMaterialFunctionCall_5 {
   class=MaterialExpressionMaterialFunctionCall
   inputs [
    "Xb1,PS4" {from=MaterialExpressionMaterialFunctionCall_6, output=Result}
    "PC-High&&Epic,XSX,PS5" {from=MaterialExpressionMaterialFunctionCall_6, output=Result}
    "PC-Perf,Android" {from=MaterialExpressionMaterialFunctionCall_13, output=Result}
    Switch {from=MaterialExpressionMaterialFunctionCall_13, output=Result}
    "PC-Low(NotPerf)&&Medium" {from=MaterialExpressionMaterialFunctionCall_6, output=Result}
   ]
   properties {materialFunction=/Game/Effects/Niagara/MasterMaterials/Functions/Common/PlatformSwitch_3D}
  }
  MaterialExpressionMaterialFunctionCall_6 {
   class=MaterialExpressionMaterialFunctionCall
   inputs [Emissive {from=MaterialExpressionMultiply_21}]
   properties {materialFunction=&MF_FX_EmissiveAdjustments}
  }
  MaterialExpressionMaterialFunctionCall_8 {
   class=MaterialExpressionMaterialFunctionCall
   inputs [
    Input {from=MaterialExpressionMaterialFunctionCall_10, output=R}
    Input Low {from=MaterialExpressionConstant_8}
    Input High {from=MaterialExpressionConstant_9}
    Target Low {from=MaterialExpressionReroute_15}
    Target High {from=MaterialExpressionConstant_9}
   ]
   properties {materialFunction=&RemapValueRange}
  }
  MaterialExpressionMaterialFunctionCall_9 {
   class=MaterialExpressionMaterialFunctionCall
   inputs [
    UorV {from=MaterialExpressionConvert_3}
    Invert_EdgeControl {from=MaterialExpressionReroute_29}
    NoInvert_EdgeControl {from=MaterialExpressionVectorParameter_16, output=RGB}
   ]
   properties {materialFunction=&FXCommon_EdgeMask_Ex}
  }
  MaterialExpressionMin_0 {
   class=MaterialExpressionMin
   inputs [
    A {from=MaterialExpressionMaterialFunctionCall_0, output=Mask}
    B {from=MaterialExpressionMaterialFunctionCall_2, output=Mask}
   ]
   properties {}
  }
  MaterialExpressionMultiply_17 {
   class=MaterialExpressionMultiply
   inputs [
    A {from=MaterialExpressionMaterialFunctionCall_11, output=Result}
    B {from=MaterialExpressionLinearInterpolate_0}
   ]
  }
  MaterialExpressionMultiply_18 {
   class=MaterialExpressionMultiply
   inputs [
    A {from=MaterialExpressionMultiply_25}
    B {from=MaterialExpressionVertexNormalWS_1}
   ]
  }
  MaterialExpressionMultiply_19 {
   class=MaterialExpressionMultiply
   inputs [
    A {from=MaterialExpressionReroute_28}
    B {from=""}
   ]
   properties {constB=0.5732}
  }
  MaterialExpressionMultiply_2 {
   class=MaterialExpressionMultiply
   inputs [
    A {from=""}
    B {from=MaterialExpressionMin_0}
   ]
   properties {constA=1}
  }
  MaterialExpressionMultiply_20 {
   class=MaterialExpressionMultiply
   inputs [
    A {from=MaterialExpressionPower_2}
    B {from=MaterialExpressionMaterialFunctionCall_8, output=Result}
   ]
  }
  MaterialExpressionMultiply_21 {
   class=MaterialExpressionMultiply
   inputs [
    A {from=MaterialExpressionFresnel_1}
    B {from=MaterialExpressionMultiply_23}
   ]
  }
  MaterialExpressionMultiply_22 {
   class=MaterialExpressionMultiply
   inputs [
    A {from=MaterialExpressionMultiply_28}
    B {from=MaterialExpressionParticleColor_1, output=A}
   ]
  }
  MaterialExpressionMultiply_23 {
   class=MaterialExpressionMultiply
   inputs [
    A {from=MaterialExpressionSkyLightEnvMapSample_1}
    B {from=MaterialExpressionReroute_22}
   ]
  }
  MaterialExpressionMultiply_24 {
   class=MaterialExpressionMultiply
   inputs [
    A {from=MaterialExpressionSaturate_1}
    B {from=MaterialExpressionReroute_25}
   ]
  }
  MaterialExpressionMultiply_25 {
   class=MaterialExpressionMultiply
   inputs [
    A {from=MaterialExpressionMultiply_20}
    B {from=MaterialExpressionDynamicParameter_2, output=Param1}
   ]
  }
  MaterialExpressionMultiply_26 {
   class=MaterialExpressionMultiply
   inputs [
    A {from=MaterialExpressionAdd_5}
    B {from=MaterialExpressionScalarParameter_2}
   ]
  }
  MaterialExpressionMultiply_27 {
   class=MaterialExpressionMultiply
   inputs [
    A {from=MaterialExpressionAdd_6}
    B {from=MaterialExpressionScalarParameter_9}
   ]
  }
  MaterialExpressionMultiply_28 {
   class=MaterialExpressionMultiply
   inputs [
    A {from=MaterialExpressionDepthFade_1}
    B {from=MaterialExpressionMultiply_24}
   ]
  }
  MaterialExpressionMultiply_29 {
   class=MaterialExpressionMultiply
   inputs [
    A {from=MaterialExpressionScalarParameter_14}
    B {from=MaterialExpressionAdd_4}
   ]
  }
  MaterialExpressionMultiply_3 {
   class=MaterialExpressionMultiply
   inputs [
    A {from=MaterialExpressionReroute_11}
    B {from=MaterialExpressionMultiply_2}
   ]
  }
  MaterialExpressionMultiply_30 {
   class=MaterialExpressionMultiply
   inputs [
    A {from=MaterialExpressionMaterialFunctionCall_9, output=Mask}
    B {from=MaterialExpressionComponentMask_3}
   ]
  }
  MaterialExpressionMultiply_31 {
   class=MaterialExpressionMultiply
   inputs [
    A {from=MaterialExpressionMultiply_32}
    B {from=MaterialExpressionFresnel_1}
   ]
  }
  MaterialExpressionMultiply_32 {
   class=MaterialExpressionMultiply
   inputs [
    A {from=MaterialExpressionReroute_23}
    B {from=""}
   ]
   properties {constB=0.25}
  }
  MaterialExpressionMultiply_33 {
   class=MaterialExpressionMultiply
   inputs [
    A {from=MaterialExpressionSaturate_6}
    B {from=MaterialExpressionSaturate_6}
   ]
  }
  MaterialExpressionParticleColor_1 {class=MaterialExpressionParticleColor, inputs [], properties {}}
  MaterialExpressionPower_0 {
   class=MaterialExpressionPower
   inputs [
    Base {from=MaterialExpressionSaturate_0}
    Exp {from=MaterialExpressionScalarParameter_4}
   ]
  }
  MaterialExpressionPower_2 {
   class=MaterialExpressionPower
   inputs [
    Base {from=MaterialExpressionReroute_30}
    Exp {from=MaterialExpressionConstant_13}
   ]
  }
  MaterialExpressionReflectionVectorWS_1 {class=MaterialExpressionReflectionVectorWS, inputs [CustomWorldNormal {from=MaterialExpressionReroute_8, output=""}], properties {}}
  MaterialExpressionReroute_10 {class=MaterialExpressionReroute, inputs [None {from=MaterialExpressionMaterialFunctionCall_5, output=Result}]}
  MaterialExpressionReroute_11 {class=MaterialExpressionReroute, inputs [None {from=MaterialExpressionReroute_7}]}
  MaterialExpressionReroute_12 {class=MaterialExpressionReroute, inputs [None {from=MaterialExpressionComponentMask_2}]}
  MaterialExpressionReroute_13 {class=MaterialExpressionReroute, inputs [None {from=MaterialExpressionMultiply_18}]}
  MaterialExpressionReroute_14 {class=MaterialExpressionReroute, inputs [None {from=MaterialExpressionVectorParameter_4, output=RGB}]}
  MaterialExpressionReroute_15 {class=MaterialExpressionReroute, inputs [None {from=MaterialExpressionScalarParameter_0}]}
  MaterialExpressionReroute_16 {class=MaterialExpressionReroute, inputs [None {from=MaterialExpressionCustom_1}]}
  MaterialExpressionReroute_17 {class=MaterialExpressionReroute, inputs [None {from=MaterialExpressionVectorParameter_12, output=R}]}
  MaterialExpressionReroute_18 {class=MaterialExpressionReroute, inputs [None {from=MaterialExpressionVectorParameter_12, output=G}]}
  MaterialExpressionReroute_19 {class=MaterialExpressionReroute, inputs [None {from=MaterialExpressionVectorParameter_12, output=B}]}
  MaterialExpressionReroute_20 {class=MaterialExpressionReroute, inputs [None {from=MaterialExpressionVectorParameter_12, output=A}]}
  MaterialExpressionReroute_21 {class=MaterialExpressionReroute, inputs [None {from=MaterialExpressionSaturate_0}]}
  MaterialExpressionReroute_22 {class=MaterialExpressionReroute, inputs [None {from=MaterialExpressionVectorParameter_14, output=RGB}]}
  MaterialExpressionReroute_23 {class=MaterialExpressionReroute, inputs [None {from=MaterialExpressionVectorParameter_14, output=RGB}]}
  MaterialExpressionReroute_24 {class=MaterialExpressionReroute, inputs [None {from=MaterialExpressionVectorParameter_15, output=RGB}]}
  MaterialExpressionReroute_25 {class=MaterialExpressionReroute, inputs [None {from=MaterialExpressionScalarParameter_11}]}
  MaterialExpressionReroute_26 {class=MaterialExpressionReroute, inputs [None {from=MaterialExpressionScalarParameter_13}]}
  MaterialExpressionReroute_27 {class=MaterialExpressionReroute, inputs [None {from=MaterialExpressionReroute_8}]}
  MaterialExpressionReroute_28 {class=MaterialExpressionReroute, inputs [None {from=MaterialExpressionDynamicParameter_3, output=Param4}]}
  MaterialExpressionReroute_29 {class=MaterialExpressionReroute, inputs [None {from=MaterialExpressionVectorParameter_17, output=RGB}]}
  MaterialExpressionReroute_30 {class=MaterialExpressionReroute, inputs [None {from=MaterialExpressionMultiply_30}]}
  MaterialExpressionReroute_31 {class=MaterialExpressionReroute, inputs [None {from=MaterialExpressionMultiply_30}]}
  MaterialExpressionReroute_32 {class=MaterialExpressionReroute, inputs [None {from=MaterialExpressionVectorParameter_18, output=RGB}]}
  MaterialExpressionReroute_33 {class=MaterialExpressionReroute, inputs [None {from=MaterialExpressionVectorParameter_18, output=VPan}]}
  MaterialExpressionReroute_34 {class=MaterialExpressionReroute, inputs [None {from=MaterialExpressionVectorParameter_19, output=RGB}]}
  MaterialExpressionReroute_35 {class=MaterialExpressionReroute, inputs [None {from=MaterialExpressionVectorParameter_20, output=RGB}]}
  MaterialExpressionReroute_36 {class=MaterialExpressionReroute, inputs [None {from=MaterialExpressionVectorParameter_21, output=RGB}]}
  MaterialExpressionReroute_7 {class=MaterialExpressionReroute, inputs [None {from=MaterialExpressionMultiply_22}]}
  MaterialExpressionReroute_8 {class=MaterialExpressionReroute, inputs [None {from=MaterialExpressionTransform_1}]}
  MaterialExpressionReroute_9 {class=MaterialExpressionReroute, inputs [None {from=MaterialExpressionMultiply_17}]}
  MaterialExpressionSaturate_0 {class=MaterialExpressionSaturate, inputs [None {from=MaterialExpressionMultiply_26}]}
  MaterialExpressionSaturate_1 {class=MaterialExpressionSaturate, inputs [None {from=MaterialExpressionReroute_21}]}
  MaterialExpressionSaturate_6 {class=MaterialExpressionSaturate, inputs [None {from=MaterialExpressionMultiply_27}]}
  MaterialExpressionSaturate_7 {class=MaterialExpressionSaturate, inputs [None {from=MaterialExpressionMultiply_33}]}
  MaterialExpressionScalarParameter_0 {class=MaterialExpressionScalarParameter, properties {defaultValue=0.6, parameterName=AdditionNoise_LowestV}}
  MaterialExpressionScalarParameter_1 {class=MaterialExpressionScalarParameter, properties {group=TotalOpacity, parameterName=TotalOpacity_GrayAdjust_Add}}
  MaterialExpressionScalarParameter_11 {class=MaterialExpressionScalarParameter, properties {defaultValue=0.5, parameterName=TotalOpacityScale}}
  MaterialExpressionScalarParameter_13 {class=MaterialExpressionScalarParameter, properties {defaultValue=200, parameterName=DepthFade}}
  MaterialExpressionScalarParameter_14 {class=MaterialExpressionScalarParameter, properties {defaultValue=-0.4, parameterName=Speed}}
  MaterialExpressionScalarParameter_2 {class=MaterialExpressionScalarParameter, properties {defaultValue=1.4385, group=TotalOpacity, parameterName=TotalOpacity_GrayAdjust_Mul}}
  MaterialExpressionScalarParameter_4 {class=MaterialExpressionScalarParameter, properties {defaultValue=1, group=TotalOpacity, parameterName=TotalOpacity_GrayAdjust_Power}}
  MaterialExpressionScalarParameter_5 {class=MaterialExpressionScalarParameter, properties {defaultValue=-0.2432, group=ColorG, parameterName=ColorG_GrayAdjust_Add}}
  MaterialExpressionScalarParameter_9 {class=MaterialExpressionScalarParameter, properties {defaultValue=0.8256, group=ColorG, parameterName=ColorG_GrayAdjust_Mul}}
  MaterialExpressionSkyLightEnvMapSample_1 {
   class=MaterialExpressionSkyLightEnvMapSample
   inputs [
    Direction {from=MaterialExpressionReflectionVectorWS_1, output=""}
    Roughness {from=MaterialExpressionConstant_12, output=""}
   ]
   properties {}
  }
  MaterialExpressionTextureCoordinate_0 {class=MaterialExpressionTextureCoordinate}
  MaterialExpressionTextureCoordinate_2 {class=MaterialExpressionTextureCoordinate}
  MaterialExpressionTextureCoordinate_3 {class=MaterialExpressionTextureCoordinate}
  MaterialExpressionTextureObjectParameter_1 {
   class=MaterialExpressionTextureObjectParameter
   inputs [
    UVs {from="", output=""}
    Apply View MipBias {from="", output=""}
   ]
   properties {
    group=AdditionNoise
    parameterName=AdditionNoise_Texture
    samplerType=SAMPLERTYPE_Masks
    sortPriority=0
    texture=/FortWater/Textures/T_OceanFoam_New_Mask
   }
  }
  MaterialExpressionTime_1 {class=MaterialExpressionTime, properties {bOverride_Period=true, period=6000}}
  MaterialExpressionTransform_1 {class=MaterialExpressionTransform, inputs [None {from=MaterialExpressionComponentMask_2}]}
  MaterialExpressionVectorParameter_1 {
   class=MaterialExpressionVectorParameter
   properties {group=Mask, parameterName=MaskV_EdgeMask_Control, channelNames {a="", b=Power, g=Multiply, r=Offset}, defaultValue=<rgba -0.1256 10 2 1>}
  }
  MaterialExpressionVectorParameter_12 {class=MaterialExpressionVectorParameter, properties {parameterName=UV_Falloff_NormalStrength, defaultValue=<rgba 1 2.5 1 1>}}
  MaterialExpressionVectorParameter_13 {class=MaterialExpressionVectorParameter, properties {parameterName=AColor, defaultValue=<rgba 0 0.1322 0.2 1>}}
  MaterialExpressionVectorParameter_14 {class=MaterialExpressionVectorParameter, properties {parameterName=ReflectionColor, defaultValue=<rgba 0.5515 0.35 0.7 1>}}
  MaterialExpressionVectorParameter_15 {class=MaterialExpressionVectorParameter, properties {parameterName=BColor, defaultValue=<rgba 0 2 1.4261 1>}}
  MaterialExpressionVectorParameter_16 {
   class=MaterialExpressionVectorParameter
   properties {
    group=WPOEdgeMask
    parameterName=WPOEdgeMask_EdgeMask_Control
    channelNames {a="", b=Power, g=Multiply, r=Offset}
    defaultValue=<rgba 0 2 1 1>
   }
  }
  MaterialExpressionVectorParameter_17 {
   class=MaterialExpressionVectorParameter
   properties {
    group=WPOEdgeMask
    parameterName=WPOEdgeMask_EdgeMask_InvertControl
    channelNames {a="", b=Power, g=Multiply, r=Offset}
    defaultValue=<rgba 0 2 1 1>
   }
  }
  MaterialExpressionVectorParameter_18 {
   class=MaterialExpressionVectorParameter
   properties {
    group=AdditionNoise
    parameterName=AdditionNoise_UVControl
    channelNames {a=VPan, b=UPan, g=VScale, r=UScale}
    defaultValue=<rgba 0.15 0.15 0.3 0>
   }
  }
  MaterialExpressionVectorParameter_19 {class=MaterialExpressionVectorParameter, properties {parameterName="Color_Morning (V3)", defaultValue=<rgba 0.85 0.85 0.85 0>}}
  MaterialExpressionVectorParameter_2 {
   class=MaterialExpressionVectorParameter
   properties {
    group=Mask
    parameterName=MaskV_EdgeMask_InvertControl
    channelNames {a="", b=Power, g=Multiply, r=Offset}
    defaultValue=<rgba -0.2216 10 2 1>
   }
  }
  MaterialExpressionVectorParameter_20 {class=MaterialExpressionVectorParameter, properties {parameterName="Color_Day (V3)", defaultValue=<rgba 0.95 0.95 0.95 0>}}
  MaterialExpressionVectorParameter_21 {class=MaterialExpressionVectorParameter, properties {parameterName="Color_Evening (V3)", defaultValue=<rgba 0.85 0.85 0.85 0>}}
  MaterialExpressionVectorParameter_4 {class=MaterialExpressionVectorParameter, properties {parameterName="Color_Night (V3)", defaultValue=<rgba 0.75 0.75 0.75 0>}}
  MaterialExpressionVectorParameter_8 {
   class=MaterialExpressionVectorParameter
   properties {group=Mask, parameterName=Mask_EdgeMask_Control, channelNames {a="", b=Power, g=Multiply, r=Offset}, defaultValue=<rgba -0.0872 10 2 1>}
  }
  MaterialExpressionVertexNormalWS_1 {class=MaterialExpressionVertexNormalWS, inputs [], properties {}}
 }
 properties {
  bAutomaticallySetUsageInEditor=false
  bUsedWithNiagaraMeshParticles=true
  blendMode=BLEND_Translucent
  previewMesh=&/SM_WaveMesh_Cut
  translucencyLightingMode=TLM_VolumetricDirectional
  usedShadingModels=Default Lit
 }
}
