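# Niagara system for the water disturbance at the Zero Point environment effect, layering drops, mist, splashes and waves.
# Sprite emitters spawn over wide shapes with point-based velocities, drag, SubUV animation and curve-driven material parameters, several on GPU.
# A wave emitter places oriented, curve-scaled particles on a flattened torus moved by point force.
# Camera distance drives some sprite sizes and spawn rates, and translucent splashes are tinted by time of day.
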NiagaraSystem /Game/Environments/Hera/FX/ZeroPoint/WaterDisturbance/NS_Env_ZeroPoint_Water
dependencies {
 usedDataInterfaces [
  &NiagaraDataInterfaceCurve
  &NiagaraDataInterfaceCamera
  &NiagaraDataInterfaceSpriteRendererInfo
  &NiagaraDataInterfaceColorCurve
 ]
 usedDynamicInputs [&MakeFloatFromVector2d]
 usedModules [&SystemState]
}
emitters {
 Water_Drops {
  emitterUpdateScript [
   &EmitterState()
   &SpawnRate(SpawnRate=12, Use Spawn Probability=false)
  ]
  particleSpawnScript [
   &InitializeParticle(
    Color Mode=::NewEnumerator0
    Lifetime=&RandomRangeFloat(Minimum=2.5, Maximum=3)
    Sprite Rotation Mode=::NewEnumerator2
    Sprite Size Mode=::NewEnumerator3
    Uniform Sprite Size=&FloatFromCurve(
     FloatCurve=DataInterface() {propertyValues {Curve {keys [<at 0.0081 0.25>,<at 0.93 0.55>,<at 0.97 0.9446>,<at 1 1>]}}}
     CurveIndex=&RandomRangeFloat(Minimum=0.45, Maximum=1)
     Scale Curve=13
    )
    UsePositionOffset=false
    Write Material Random=true
   )
   &ShapeLocation(
    Handle Radius=99
    Large Radius=800
    MaskSpawnGroup=false
    Non Uniform Scale=<1 1 4.3>
    Offset=<0 0 441.5258>
    Offset Mode=::NewEnumerator0
    Override Seed=false
    Shape Primitive=::NewEnumerator3
    UseOwnerScale=true
   )
   &AddVelocity(
    Module.Use Falloff Distance=false
    Origin Offset=<0 0 19513.8828>
    Override Seed=false
    Velocity Mode=::NewEnumerator1
    Velocity Speed=&RandomRangeFloat(Minimum=-182.8822, Maximum=-367.0504)
   )
  ]
  particleUpdateScript [
   &ParticleState()
   &ScaleSpriteSize(
    Scale Sprite Size Mode=::NewEnumerator0
    Uniform Scale Factor=&RandomRangeFloat(Minimum=0.5, Maximum=1.1, Evaluation Type=::NewEnumerator1)
   ) {module=ScaleSpriteSize002}
   &SpriteRotationRate(Rotation Rate=&RandomRangeFloat(Minimum=10, Maximum=20))
   &ScaleSpriteSize(
    Scale Factor=&MakeVector2D(
     X=&RandomRangeFloat(Minimum=0.9, Maximum=1, Evaluation Type=::NewEnumerator1)
     Y=&RandomRangeFloat(Minimum=1.1, Maximum=1.15, Evaluation Type=::NewEnumerator1)
    )
    Scale Sprite Size Mode=::NewEnumerator2
   ) {module=ScaleSpriteSize001}
   &ScaleSpriteSize(
    Scale Sprite Size Mode=::NewEnumerator0
    Uniform Scale Factor=&FloatFromCurve(
     FloatCurve=DataInterface() {propertyValues {Curve {
        defaultValue=0
        keys [
         {arriveTangent=7.3453, interpMode=RCIM_Cubic, leaveTangent=7.3453, tangentMode=RCTM_User, time=0, value=0}
         {interpMode=RCIM_Cubic, tangentMode=RCTM_User, time=0.5, value=1}
         {tangentMode=RCTM_User, time=1, value=0}
        ]
       }}}
     Scale Curve=5.8314
    )
   )
   &SolveForcesAndVelocity(Clamp Velocity=false, Limit Acceleration=false)
   &ScaleColor(
    Scale Alpha=&FloatFromCurve(
     FloatCurve=DataInterface() {propertyValues {Curve {
        defaultValue=1
        keys [
         {interpMode=RCIM_Cubic, leaveTangent=7.4172, tangentMode=RCTM_Break, time=0, value=0}
         {interpMode=RCIM_Cubic, tangentMode=RCTM_User, time=0.5, value=1}
         {arriveTangent=0.2381, interpMode=RCIM_Cubic, leaveTangent=0.2381, tangentMode=RCTM_User, time=0.8196, value=1}
         {tangentMode=RCTM_User, time=1, value=0}
        ]
       }}}
     Scale Curve=&RandomRangeFloat(Minimum=1, Maximum=1)
    )
    ScaleA=true
    ScaleRGB=false
   )
   &DynamicMaterialParameters(
    Index 0 Param 1=&RandomRangeFloat(Minimum=0.35, Maximum=0.5)
    Index 0 Param 2=&RandomRangeFloat(Minimum=0, Maximum=1)
    Module.Index0Float0Write=true
    Module.Index0Float1Write=true
    Module.Index0Float2Write=true
    Module.Index0Float3Write=true
   )
   &ScaleSpriteSize(
    Scale Sprite Size Mode=::NewEnumerator0
    Uniform Scale Factor=&Power(
     a=/Game/Effects/Niagara/Modules/Math/FN_CameraDistance(
      0Dist:NiagaraFloat=0
      1Dist:NiagaraFloat=2500
      Camera Query:NiagaraDataInterfaceCamera=DataInterface()
      Position:NiagaraPosition=$Particles.Position
     )
     b=1.5
    )
   ) {module=ScaleSpriteSize003}
  ]
  rendererClasses [&NiagaraSpriteRendererProperties]
  renderers [{rendererClass=&NiagaraSpriteRendererProperties}]
 }
 Water_Drops001 {
  simTarget=GPUComputeSim
  emitterUpdateScript [
   &EmitterState()
   &SpawnRate(SpawnRate=325.6253, Use Spawn Probability=false)
  ]
  particleSpawnScript [
   &InitializeParticle(
    Color Mode=::NewEnumerator0
    Lifetime=&RandomRangeFloat(Minimum=0.25, Maximum=0.35)
    Sprite Size Mode=::NewEnumerator3
    Uniform Sprite Size=&Multiply_Float(A=&RandomRangeFloat(Minimum=2, Maximum=6), B=1)
    UsePositionOffset=false
    Write Material Random=true
   )
   &ShapeLocation(
    Handle Radius=99
    Large Radius=1681.7601
    MaskSpawnGroup=false
    Offset=<0 0 90>
    Offset Mode=::NewEnumerator0
    Override Seed=false
    Shape Primitive=::NewEnumerator3
    UseOwnerScale=true
   )
   &AddVelocity(
    Module.Use Falloff Distance=false
    Origin Offset=<0 0 2311.6938>
    Override Seed=false
    Velocity Mode=::NewEnumerator1
    Velocity Speed=&RandomRangeFloat(Minimum=-1000.9698, Maximum=-2000)
   )
  ]
  particleUpdateScript [
   &ParticleState()
   &ScaleSpriteSize(
    Scale Sprite Size Mode=::NewEnumerator0
    Uniform Scale Factor=&Power(
     a=&FN_CameraDistance(
      0Dist:NiagaraFloat=0
      1Dist:NiagaraFloat=1000
      Camera Query:NiagaraDataInterfaceCamera=DataInterface()
      Position:NiagaraPosition=$Particles.Position
     )
     b=3
    )
   )
   &Drag(
    Drag=&FloatFromCurve(FloatCurve=DataInterface() {propertyValues {Curve {
        defaultValue=1
        keys [{interpMode=RCIM_Cubic, tangentMode=RCTM_User, time=0.4, value=1}, {interpMode=RCIM_Cubic, tangentMode=RCTM_User, time=1, value=2}]
       }}})
    Use Linear Drag=true
    Use Rotational Drag=true
   )
   &SolveForcesAndVelocity(Clamp Velocity=false, Limit Acceleration=false)
   &ScaleSpriteSizeBySpeed(Max Scale Factor=<2 8>)
   &ScaleColor(
    Scale Alpha=&FloatFromCurve(
     FloatCurve=DataInterface() {propertyValues {Curve {
        defaultValue=1
        keys [{interpMode=RCIM_Cubic, tangentMode=RCTM_User, time=0.8, value=1}, {interpMode=RCIM_Cubic, tangentMode=RCTM_User, time=1, value=0}]
       }}}
     Scale Curve=&RandomRangeFloat(Minimum=0.6, Maximum=1)
    )
    ScaleA=true
    ScaleRGB=false
   )
  ]
  rendererClasses [&NiagaraSpriteRendererProperties]
  renderers [{rendererClass=&NiagaraSpriteRendererProperties}]
 }
 Water_Mist {
  simTarget=GPUComputeSim
  emitterUpdateScript [
   &EmitterState()
   &SpawnRate(SpawnRate=100, Use Spawn Probability=false) {enabled=false}
   &SpawnRate(
    SpawnRate=&FN_LOD_Distance_Remap(Start Distance=1000, End Distance=0, Minimum Output=20, Maximum Output=100, Clamp Output=true)
    Use Spawn Probability=false
   ) {module=SpawnRate001}
  ]
  particleSpawnScript [
   &InitializeParticle(
    Color Mode=::NewEnumerator0
    Lifetime Max=0.7
    Lifetime Min=0.5
    Lifetime Mode=::NewEnumerator1
    Sprite Rotation Mode=::NewEnumerator2
    Sprite Size Mode=::NewEnumerator1
    Uniform Sprite Size Max=650
    Uniform Sprite Size Min=550
    UsePositionOffset=false
    Write Material Random=true
   ) {module=InitializeParticle001}
   /SetVariables(Particles.RandomFrame=&RandomRangeInt(Minimum=0, Maximum=63)) {isSetParameters=true}
   &DynamicMaterialParameters(
    Index 0 Param 1=&RandomRangeFloat(Minimum=0.2, Maximum=0.5)
    Module.Param0WriteEnabled=true
    Module.Param1WriteEnabled=true
    Module.Param2WriteEnabled=true
    Module.Param3WriteEnabled=true
   )
   &ShapeLocation(
    Handle Radius=706.286
    Large Radius=769.5361
    MaskSpawnGroup=false
    Offset=<0 0 -111>
    Offset Coordinate Space=::World
    Offset Mode=::NewEnumerator0
    Override Seed=false
    Shape Primitive=::NewEnumerator3
    UseOwnerScale=true
   )
   &AddVelocity(Module.Use Falloff Distance=false, Origin Offset=<0 0 4000>, Override Seed=false, Velocity Mode=::NewEnumerator1, Velocity Speed=-1543.1307)
   &CameraOffset(Camera Offset Amount=&RandomRangeFloat(Minimum=2, Maximum=25))
   &FN_SubUVAnimation(SubUV Animation Mode:ENiagara_SubUVLookupMode=::NewEnumerator2)
  ]
  particleUpdateScript [
   &Drag(Use Linear Drag=true, Use Rotational Drag=true)
   &GravityForce(Gravity=<0 0 1125>)
   &SolveForcesAndVelocity(Clamp Velocity=false, Limit Acceleration=false)
   &ParticleState()
   &ScaleSpriteSize(
    Scale Factor=&Multiply_Vector2D(
     A=&Vector2DFromFloat(Value=&FloatFromCurve(FloatCurve=DataInterface() {propertyValues {Curve {keys [
           {arriveTangent=1.1335, interpMode=RCIM_Cubic, leaveTangent=1.1335, tangentMode=RCTM_User, time=0, value=0.5}
           {arriveTangent=0.0897, interpMode=RCIM_Cubic, leaveTangent=0.0897, tangentMode=RCTM_User, time=1, value=1}
          ]}}}))
     B=&Vector2DFromFloat(Value=&FloatFromCurve(
       FloatCurve=DataInterface() {propertyValues {Curve {
          defaultValue=1
          keys [
           {arriveTangent=2.2913, interpMode=RCIM_Cubic, leaveTangent=2.2913, tangentMode=RCTM_User, time=0, value=0}
           {interpMode=RCIM_Cubic, tangentMode=RCTM_User, time=1, value=1}
          ]
         }}}
       CurveIndex=&FN_CameraDistance(
        0Dist:NiagaraFloat=500
        1Dist:NiagaraFloat=2000
        Camera Query:NiagaraDataInterfaceCamera=DataInterface()
        Position:NiagaraPosition=$Particles.Position
       )
      ))
    )
    Scale Sprite Size Mode=::NewEnumerator2
   )
   /Niagara/Modules/Update/SubUV/SubUVAnimation(Number Of Frames=&FloatFromCurve(FloatCurve=DataInterface() {propertyValues {Curve {keys [<at -0.0046 0>,<at 1.0011 64>]}}}))
   &ScaleColor(
    Scale Alpha=&FloatFromCurve(
     FloatCurve=DataInterface() {propertyValues {Curve {keys [<at 0 0 RCIM_Cubic>, {arriveTangent=0.002, interpMode=RCIM_Cubic, leaveTangent=0.002, time=0.16, value=1}, <at 1.0011 0.002 RCIM_Cubic>]}}}
     Scale Curve=&RandomRangeFloat(Minimum=2, Maximum=4)
    )
    ScaleA=true
    ScaleRGB=true
   )
   &DynamicMaterialParameters(
    Index 0 Param 1=&UniformRangedFloat(Minimum=1, Maximum=2)
    Index 0 Param 2=4
    Index 0 Param 3=2
    Index 0 Param 4=0.01
    Module.Index0Float0Write=true
    Module.Index0Float1Write=true
    Module.Index0Float2Write=true
    Module.Index0Float3Write=true
   ) {module=DynamicMaterialParameters001}
  ]
  rendererClasses [&NiagaraSpriteRendererProperties]
  renderers [{rendererClass=&NiagaraSpriteRendererProperties}]
 }
 Water_Splashes {
  emitterUpdateScript [
   &EmitterState()
   &SpawnRate(SpawnRate=40, Use Spawn Probability=false)
  ]
  particleSpawnScript [
   &InitializeParticle(
    Color=<rgba 87.5 91.0352 100 1>
    Lifetime Max=0.75
    Lifetime Min=0.65
    Lifetime Mode=::NewEnumerator1
    Sprite Rotation Mode=::NewEnumerator2
    Sprite Size Mode=::NewEnumerator3
    Uniform Sprite Size=&Multiply_Float(A=&RandomRangeFloat(Minimum=61.509, Maximum=83.8389), B=5)
    UsePositionOffset=false
    Write Material Random=true
   )
   &ShapeLocation(
    Large Radius=800
    MaskSpawnGroup=false
    Offset=&RandomRangeVector(Minimum=<0 0 -100>, Maximum=<0 0 100>)
    Offset Mode=::NewEnumerator0
    Override Seed=false
    Shape Primitive=::NewEnumerator3
    UseOwnerScale=true
   )
   &AddVelocity(
    Module.Use Falloff Distance=false
    Origin Offset=<0 0 -4907.7197>
    Override Seed=false
    Velocity Mode=::NewEnumerator1
    Velocity Speed=&RandomRangeFloat(Minimum=318.4366, Maximum=999)
   )
   &DynamicMaterialParameters(
    Index 0 Param 1=&FloatFromCurve(FloatCurve=DataInterface() {propertyValues {Curve {keys [
         {interpMode=RCIM_Cubic, tangentMode=RCTM_User, time=0, value=0.5}
         {arriveTangent=-2.1889, interpMode=RCIM_Cubic, leaveTangent=-2.1889, tangentMode=RCTM_User, time=1, value=7.0866}
        ]}}})
    Index 0 Param 2=0.8
    Module.Index0Param0Write=true
    Module.Index0Param1Write=true
    Module.Index0Param2Write=true
    Module.Index0Param3Write=true
   ) {module=DynamicMaterialParameters001}
  ]
  particleUpdateScript [
   &ParticleState()
   &SpriteRotationRate(Rotation Rate=&RandomRangeFloat(Minimum=-5, Maximum=5))
   &ScaleSpriteSize(
    Uniform Curve Scale=0.5
    Uniform Curve Sprite Scale=DataInterface() {propertyValues {Curve {keys [
        {arriveTangent=1.987, interpMode=RCIM_Cubic, leaveTangent=1.987, tangentMode=RCTM_User, time=0.0021, value=0.5588}
        {tangentMode=RCTM_User, time=1, value=1.25}
       ]}}}
   )
   &ScaleColor(
    Scale Alpha=&FloatFromCurve(FloatCurve=DataInterface() {propertyValues {Curve {keys [
         {interpMode=RCIM_Cubic, tangentMode=RCTM_Break, time=0, value=0}
         {interpMode=RCIM_Cubic, tangentMode=RCTM_Break, time=0.1, value=1}
         {arriveTangent=-2, interpMode=RCIM_Cubic, leaveTangent=-2, time=0.4497, value=0.95}
         {arriveTangent=-0.9842, interpMode=RCIM_Cubic, leaveTangent=-0.9842, tangentMode=RCTM_User, time=0.6, value=0}
        ]}}})
    ScaleA=true
    ScaleRGB=true
   ) {module=ScaleColor001}
   /Niagara/Modules/Update/SubUV/V2/SubUVAnimation(
    Evaluation Type=::NewEnumerator0
    "Sprite Renderer "=DataInterface() {propertyValues {SpriteRenderer=&self:Water_Splashes_1.NiagaraSpriteRendererProperties_0}}
    SubUV Animation Mode:ENiagara_SubUVLookupModeV2=::NewEnumerator2
    UseEndFrame=false
    UseStartFrame=false
   )
   &GravityForce(Gravity=<0 0 -1376.1976>)
   &Drag(
    Drag=&FloatFromCurve(FloatCurve=DataInterface() {propertyValues {Curve {keys [<at 0 0>,<at 0.25 1>]}}})
    Use Linear Drag=true
    Use Rotational Drag=true
   )
   &SolveForcesAndVelocity(Clamp Velocity=false, Limit Acceleration=false)
   &ScaleSpriteSizeBySpeed(
    Max Scale Factor=<4 4>
    Sample Scale Factor By Curve=true
    Scale Factor Curve=DataInterface() {propertyValues {Curve {keys [<at 0.7149 0.0353>,<at 0.7967 0.8118>]}}}
    Velocity Threshold=100
   )
   &DynamicMaterialParameters(
    Index 0 Param 1=&FloatFromCurve(FloatCurve=DataInterface() {propertyValues {Curve {keys [<at 0 0>, {arriveTangent=8.3645, leaveTangent=8.3645, time=0.3771, value=1.192}, <at 0.6568 6.6648>, <at 1 8.757>]}}})
    Index 0 Param 2=0.8
    Index 1 Param 1=&FloatFromCurve(FloatCurve=DataInterface() {propertyValues {Curve {keys [<at 0.4 1>,<at 0.7451 0>]}}})
    Module.Index0Param0Write=true
    Module.Index0Param1Write=true
    Module.Index0Param2Write=true
    Module.Index0Param3Write=true
    Module.Index1Param0Write=true
    Module.Index1Param1Write=true
    Module.Index1Param2Write=true
    Module.Index1Param3Write=true
    Write Parameter Index 1=true
   ) {enabled=false}
   &DynamicMaterialParameters(
    Index 0 Param 1=&FloatFromCurve(
     FloatCurve=DataInterface() {propertyValues {Curve {
        defaultValue=0
        keys [
         {interpMode=RCIM_Cubic, tangentMode=RCTM_User, time=0, value=1}
         {arriveTangent=-0.1061, interpMode=RCIM_Cubic, leaveTangent=3, tangentMode=RCTM_Break, time=0.254, value=1}
         {interpMode=RCIM_Cubic, tangentMode=RCTM_User, time=0.6, value=4}
        ]
       }}}
     Scale Curve=0.7
    )
    Index 0 Param 2=0.8
    Index 0 Param 3=&Multiply_Float(
     A=&FloatFromCurve(
      FloatCurve=DataInterface() {propertyValues {Curve {keys [
          {interpMode=RCIM_Cubic, tangentMode=RCTM_User, time=0, value=1}
          {arriveTangent=0.0657, interpMode=RCIM_Cubic, leaveTangent=0.0657, tangentMode=RCTM_User, time=0.2256, value=1.0056}
          {arriveTangent=-1.2119, interpMode=RCIM_Cubic, leaveTangent=-1.2119, tangentMode=RCTM_User, time=0.3988, value=0.8}
          {arriveTangent=-0.0551, interpMode=RCIM_Cubic, leaveTangent=-0.0551, tangentMode=RCTM_User, time=0.6531, value=0.6659}
         ]}}}
      CurveIndex=$Emitter.NormalizedLoopAge
      Scale Curve=3333
     )
     B=1
    )
    Index 1 Param 1=&FloatFromCurve(FloatCurve=DataInterface() {propertyValues {Curve {
        defaultValue=0
        keys [
         {interpMode=RCIM_Cubic, tangentMode=RCTM_User, time=0, value=0.9}
         {arriveTangent=-0.014, interpMode=RCIM_Cubic, leaveTangent=-1, tangentMode=RCTM_User, time=0.2, value=0.9}
         {interpMode=RCIM_Cubic, tangentMode=RCTM_User, time=0.3, value=0.6}
         <at 1 0.2 RCIM_Cubic>
        ]
       }}})
    Module.Index0Param0Write=true
    Module.Index0Param1Write=true
    Module.Index0Param2Write=true
    Module.Index0Param3Write=true
    Module.Index1Param0Write=true
    Module.Index1Param1Write=true
    Module.Index1Param2Write=true
    Module.Index1Param3Write=true
    Write Parameter Index 1=true
   ) {module=DynamicMaterialParameters002}
  ]
  rendererClasses [&NiagaraSpriteRendererProperties]
  renderers [
   {rendererClass=&NiagaraSpriteRendererProperties}
   {rendererClass=&NiagaraSpriteRendererProperties, rendererIndex=1}
  ]
 }
 Water_Splashes_Masked {
  simTarget=GPUComputeSim
  emitterUpdateScript [
   &EmitterState()
   &SpawnRate(SpawnRate=350, Use Spawn Probability=false)
  ]
  particleSpawnScript [
   &InitializeParticle(
    Color Maximum=<rgba 0.5814 0.8292 1 1>
    Color Minimum=<rgba 0.2845 0.4342 0.5938 1>
    Color Mode=::NewEnumerator2
    Lifetime Max=1
    Lifetime Min=0.8
    Lifetime Mode=::NewEnumerator1
    Sprite Rotation Mode=::NewEnumerator2
    Sprite Size Mode=::NewEnumerator1
    Uniform Sprite Size Max=100
    Uniform Sprite Size Min=400
    UsePositionOffset=false
    Write Material Random=true
   )
   &SystemLocation()
   &ShapeLocation(
    Handle Radius=249.5957
    Large Radius=1300
    MaskSpawnGroup=false
    Offset=&RandomRangeVector(Minimum=<0 -222 -100>, Maximum=<0 222 0>)
    Offset Mode=::NewEnumerator0
    Override Seed=false
    Shape Primitive=::NewEnumerator3
    UseOwnerScale=true
   )
   &AddVelocity(
    Module.Use Falloff Distance=false
    Origin Offset=<0 0 4000>
    Override Seed=false
    Velocity Mode=::NewEnumerator1
    Velocity Speed=&RandomRangeFloat(Minimum=-2075.1338, Maximum=-900)
   ) {module=AddVelocity001}
   &ApplyInitialForces()
   &CameraOffset(Camera Offset Amount=-200) {module=CameraOffset001}
  ]
  particleUpdateScript [
   &ParticleState()
   &SpriteRotationRate(Rotation Rate=&RandomRangeFloat(Minimum=-5, Maximum=5))
   &ScaleSpriteSize(
    Uniform Curve Scale=0.5
    Uniform Curve Sprite Scale=DataInterface() {propertyValues {Curve {keys [
        {arriveTangent=1.987, interpMode=RCIM_Cubic, leaveTangent=1.987, tangentMode=RCTM_User, time=0.0021, value=0.5588}
        {tangentMode=RCTM_User, time=1, value=1.25}
       ]}}}
   )
   &ScaleColor(
    Scale Alpha=&FloatFromCurve(FloatCurve=DataInterface() {propertyValues {Curve {keys [
         {interpMode=RCIM_Cubic, tangentMode=RCTM_Break, time=0, value=0}
         {interpMode=RCIM_Cubic, tangentMode=RCTM_Break, time=0.1, value=1}
         {arriveTangent=-2, interpMode=RCIM_Cubic, leaveTangent=-2, time=0.5, value=0.95}
         {arriveTangent=-0.9842, interpMode=RCIM_Cubic, leaveTangent=-0.9842, tangentMode=RCTM_User, time=0.6, value=0}
        ]}}})
    ScaleA=true
    ScaleRGB=true
   ) {module=ScaleColor001}
   &SubUVAnimation(
    Evaluation Type=::NewEnumerator0
    "Sprite Renderer "=DataInterface() {propertyValues {SpriteRenderer=&self:Water_Splashes_Masked_1.NiagaraSpriteRendererProperties_0}}
    SubUV Animation Mode:ENiagara_SubUVLookupModeV2=::NewEnumerator2
    UseEndFrame=false
    UseStartFrame=false
   )
   &Drag(
    Drag=&FloatFromCurve(FloatCurve=DataInterface() {propertyValues {Curve {keys [<at 0 0>,<at 0.25 1>]}}}, Scale Curve=2)
    Use Linear Drag=true
    Use Rotational Drag=true
   )
   &GravityForce(Gravity=<0 0 -800>)
   &SolveForcesAndVelocity(Clamp Velocity=false, Limit Acceleration=false)
   &ScaleSpriteSizeBySpeed(
    Max Scale Factor=<4 4>
    Sample Scale Factor By Curve=true
    Scale Factor Curve=DataInterface() {propertyValues {Curve {keys [<at 0.7149 0.0353>,<at 0.7967 0.8118>]}}}
    Velocity Threshold=100
   )
   &DynamicMaterialParameters(
    Index 0 Param 1=&FloatFromCurve(FloatCurve=DataInterface() {propertyValues {Curve {keys [<at 0 0>, {arriveTangent=8.3645, leaveTangent=8.3645, time=0.3771, value=1.123}, <at 0.6568 6.6648>, <at 1 8.757>]}}})
    Index 0 Param 2=0.8
    Index 0 Param 3=&Multiply_Float(
     A=&FloatFromCurve(
      FloatCurve=DataInterface() {propertyValues {Curve {keys [
          {interpMode=RCIM_Cubic, tangentMode=RCTM_User, time=0, value=1}
          {arriveTangent=0.0657, interpMode=RCIM_Cubic, leaveTangent=0.0657, tangentMode=RCTM_User, time=0.2256, value=1.0056}
          {arriveTangent=-1.2119, interpMode=RCIM_Cubic, leaveTangent=-1.2119, tangentMode=RCTM_User, time=0.3988, value=0.8}
          {arriveTangent=-0.0551, interpMode=RCIM_Cubic, leaveTangent=-0.0551, tangentMode=RCTM_User, time=0.6531, value=0.6659}
         ]}}}
      CurveIndex=$Emitter.NormalizedLoopAge
      Scale Curve=0
     )
     B=1
    )
    Index 1 Param 1=&FloatFromCurve(FloatCurve=DataInterface() {propertyValues {Curve {keys [<at 0.4 1>,<at 0.7451 0>]}}})
    Module.Index0Param0Write=true
    Module.Index0Param1Write=true
    Module.Index0Param2Write=true
    Module.Index0Param3Write=true
    Module.Index1Param0Write=true
    Module.Index1Param1Write=true
    Module.Index1Param2Write=true
    Module.Index1Param3Write=true
    Write Parameter Index 1=true
   ) {enabled=false}
   &DynamicMaterialParameters(
    Index 0 Param 1=&FloatFromCurve(
     FloatCurve=DataInterface() {propertyValues {Curve {
        defaultValue=0
        keys [
         {interpMode=RCIM_Cubic, tangentMode=RCTM_User, time=0, value=1}
         {arriveTangent=-0.1061, interpMode=RCIM_Cubic, leaveTangent=3, tangentMode=RCTM_Break, time=0.254, value=1}
         {interpMode=RCIM_Cubic, tangentMode=RCTM_User, time=0.6, value=4}
        ]
       }}}
     Scale Curve=0.7
    )
    Index 0 Param 2=0.8
    Index 0 Param 3=&Multiply_Float(
     A=&FloatFromCurve(
      FloatCurve=DataInterface() {propertyValues {Curve {keys [
          {interpMode=RCIM_Cubic, tangentMode=RCTM_User, time=0, value=1}
          {arriveTangent=0.0657, interpMode=RCIM_Cubic, leaveTangent=0.0657, tangentMode=RCTM_User, time=0.2256, value=1.0056}
          {arriveTangent=-1.2119, interpMode=RCIM_Cubic, leaveTangent=-1.2119, tangentMode=RCTM_User, time=0.3988, value=0.8}
          {arriveTangent=-0.0551, interpMode=RCIM_Cubic, leaveTangent=-0.0551, tangentMode=RCTM_User, time=0.6531, value=0.6659}
         ]}}}
      CurveIndex=$Emitter.NormalizedLoopAge
      Scale Curve=3333
     )
     B=1
    )
    Index 1 Param 1=&FloatFromCurve(FloatCurve=DataInterface() {propertyValues {Curve {
        defaultValue=0
        keys [
         {interpMode=RCIM_Cubic, tangentMode=RCTM_User, time=0, value=1}
         {arriveTangent=-0.014, interpMode=RCIM_Cubic, leaveTangent=-1, tangentMode=RCTM_User, time=0.2, value=1}
         {interpMode=RCIM_Cubic, tangentMode=RCTM_User, time=0.3, value=0.6}
         <at 1 0.2 RCIM_Cubic>
        ]
       }}})
    Module.Index0Param0Write=true
    Module.Index0Param1Write=true
    Module.Index0Param2Write=true
    Module.Index0Param3Write=true
    Module.Index1Param0Write=true
    Module.Index1Param1Write=true
    Module.Index1Param2Write=true
    Module.Index1Param3Write=true
    Write Parameter Index 1=true
   ) {module=DynamicMaterialParameters001}
   /SetVariables(Particles.Velocity2:Vector3f=&Multiply_VectorByFloat(Vector=$Particles.Velocity, Float=30)) {isSetParameters=true, module=SetVariables_#1}
  ]
  rendererClasses [&NiagaraSpriteRendererProperties]
  renderers [{rendererClass=&NiagaraSpriteRendererProperties}]
 }
 Water_Splashes_Translucent {
  simTarget=GPUComputeSim
  emitterUpdateScript [
   &EmitterState()
   &SpawnRate(SpawnRate=333, Use Spawn Probability=false) {enabled=false}
   &SpawnRate(
    SpawnRate=&FN_LOD_Distance_Remap(Start Distance=7000, End Distance=900, Minimum Output=333, Maximum Output=33, Clamp Output=true)
    Use Spawn Probability=false
   ) {enabled=false, module=SpawnRate001}
   &SpawnRate(
    SpawnRate=&FN_LOD_Distance_Remap(Start Distance=0, End Distance=15000, Minimum Output=333, Maximum Output=25, Clamp Output=true)
    Use Spawn Probability=false
   ) {module=SpawnRate002}
  ]
  particleSpawnScript [
   &InitializeParticle(
    Color Maximum=<rgba 0.5814 0.8292 1 1>
    Color Minimum=<rgba 0.2845 0.4342 0.5938 1>
    Color Mode=::NewEnumerator2
    Lifetime Max=0.8
    Lifetime Min=0.6
    Lifetime Mode=::NewEnumerator1
    Sprite Rotation Mode=::NewEnumerator2
    Sprite Size Mode=::NewEnumerator1
    Uniform Sprite Size Max=100
    Uniform Sprite Size Min=300
    UsePositionOffset=false
    Write Material Random=true
   )
   &SystemLocation()
   &ShapeLocation(
    Handle Radius=249.5957
    Large Radius=1300
    MaskSpawnGroup=false
    Offset=&RandomRangeVector(Minimum=<0 -222 -100>, Maximum=<0 222 0>)
    Offset Mode=::NewEnumerator0
    Override Seed=false
    Shape Primitive=::NewEnumerator3
    UseOwnerScale=true
   )
   &AddVelocity(
    Module.Use Falloff Distance=false
    Origin Offset=<0 0 4000>
    Override Seed=false
    Velocity Mode=::NewEnumerator1
    Velocity Speed=&RandomRangeFloat(Minimum=-2075.1338, Maximum=-900)
   ) {module=AddVelocity001}
   &ApplyInitialForces()
  ]
  particleUpdateScript [
   &ParticleState()
   &CameraOffset(Camera Offset Amount=&FN_LOD_Distance_Remap(Start Distance=900, End Distance=15000, Minimum Output=200, Maximum Output=-200, Clamp Output=true)) {module=CameraOffset001}
   &SpriteRotationRate(Rotation Rate=&RandomRangeFloat(Minimum=-5, Maximum=5))
   &ScaleSpriteSize(
    Uniform Curve Scale=0.5
    Uniform Curve Sprite Scale=DataInterface() {propertyValues {Curve {keys [
        {arriveTangent=1.987, interpMode=RCIM_Cubic, leaveTangent=1.987, tangentMode=RCTM_User, time=0.0021, value=0.5588}
        {tangentMode=RCTM_User, time=1, value=1.25}
       ]}}}
   ) {enabled=false}
   &ScaleSpriteSize(
    Scale Sprite Size Mode=::NewEnumerator0
    Uniform Scale Factor=&Multiply_Float(
     A=&FloatFromCurve(
      FloatCurve=DataInterface() {propertyValues {Curve {keys [
          {arriveTangent=1.987, interpMode=RCIM_Cubic, leaveTangent=1.987, tangentMode=RCTM_User, time=0.0021, value=0.5588}
          {tangentMode=RCTM_User, time=1, value=1.25}
         ]}}}
      Scale Curve=0.5
     )
     B=&FloatFromCurve(
      FloatCurve=DataInterface() {propertyValues {Curve {
         defaultValue=0
         keys [
          {arriveTangent=3.1346, interpMode=RCIM_Cubic, leaveTangent=3.1346, tangentMode=RCTM_User, time=0, value=0}
          {arriveTangent=1, interpMode=RCIM_Cubic, leaveTangent=-0.3076, tangentMode=RCTM_Break, time=0.2, value=0.9971}
          {interpMode=RCIM_Cubic, tangentMode=RCTM_User, time=1, value=1}
         ]
        }}}
      CurveIndex=&FN_LOD_Distance_Remap(Start Distance=500, End Distance=15000, Minimum Output=0, Maximum Output=1, Clamp Output=true)
     )
    )
   ) {module=ScaleSpriteSize001}
   &ScaleColor(
    Scale Alpha=&FloatFromCurve(FloatCurve=DataInterface() {propertyValues {Curve {keys [
         {interpMode=RCIM_Cubic, tangentMode=RCTM_User, time=0, value=0}
         {interpMode=RCIM_Cubic, tangentMode=RCTM_User, time=0.2, value=0.2}
         {tangentMode=RCTM_User, time=1, value=0}
        ]}}})
    Scale RGB=&MakeVectorFromLinearColorRGB(Color=&ColorFromCurve(
      ColorCurve=DataInterface() {propertyValues {
        AlphaCurve {keys [<at 0 1 RCIM_Cubic>,<at 1 0 RCIM_Cubic>]}
        BlueCurve {keys [<at 0 1 RCIM_Cubic>,<at 0.5069 1>,<at 1 1 RCIM_Cubic>]}
        GreenCurve {keys [<at 0 1 RCIM_Cubic>,<at 0.5069 1>,<at 1 1 RCIM_Cubic>]}
        RedCurve {keys [<at 0 1 RCIM_Cubic>,<at 0.5069 1>,<at 1 1 RCIM_Cubic>]}
       }}
      CurveIndex=Linked() {linkedVariable {name=NPC.FortniteNPC.TimeOfDay_0-1, type=NiagaraFloat}}
     ))
    ScaleA=true
    ScaleRGB=true
   ) {enabled=false}
   &ScaleColor(
    Scale Alpha=&FloatFromCurve(
     FloatCurve=DataInterface() {propertyValues {Curve {
        defaultValue=1
        keys [
         {interpMode=RCIM_Cubic, leaveTangent=-2.0674, tangentMode=RCTM_Break, time=0.3, value=1}
         {interpMode=RCIM_Cubic, tangentMode=RCTM_User, time=0.5, value=0.3}
         <at 1 0 RCIM_Cubic>
        ]
       }}}
     Scale Curve=&RandomRangeFloat(Minimum=0.8, Maximum=1)
    )
    Scale RGB=&MakeVectorFromLinearColorRGB(Color=&ColorFromCurve(
      ColorCurve=DataInterface() {propertyValues {
        AlphaCurve {keys [<at 0 1 RCIM_Cubic>,<at 1 0 RCIM_Cubic>]}
        BlueCurve {keys [<at 0 1 RCIM_Cubic>,<at 0.5069 1>,<at 1 1 RCIM_Cubic>]}
        GreenCurve {keys [<at 0 1 RCIM_Cubic>,<at 0.5069 1>,<at 1 1 RCIM_Cubic>]}
        RedCurve {keys [<at 0 1 RCIM_Cubic>,<at 0.5069 1>,<at 1 1 RCIM_Cubic>]}
       }}
      CurveIndex=Linked() {linkedVariable {name=NPC.FortniteNPC.TimeOfDay_0-1, type=NiagaraFloat}}
     ))
    ScaleA=true
    ScaleRGB=true
   ) {enabled=false, module=ScaleColor001}
   &ScaleColor(
    Scale Alpha=&FloatFromCurve(
     FloatCurve=DataInterface() {propertyValues {Curve {
        defaultValue=1
        keys [
         {interpMode=RCIM_Cubic, leaveTangent=-2.0674, tangentMode=RCTM_Break, time=0.3, value=1}
         {interpMode=RCIM_Cubic, tangentMode=RCTM_User, time=0.4, value=0.3}
         <at 1 0 RCIM_Cubic>
        ]
       }}}
     Scale Curve=&RandomRangeFloat(Minimum=0.8, Maximum=1)
    )
    Scale RGB=&MakeVectorFromLinearColorRGB(Color=&ColorFromCurve(
      ColorCurve=DataInterface() {propertyValues {
        AlphaCurve {keys [<at 0 1 RCIM_Cubic>,<at 1 0 RCIM_Cubic>]}
        BlueCurve {keys [<at 0 1 RCIM_Cubic>,<at 0.5069 1>,<at 1 1 RCIM_Cubic>]}
        GreenCurve {keys [<at 0 1 RCIM_Cubic>,<at 0.5069 1>,<at 1 1 RCIM_Cubic>]}
        RedCurve {keys [<at 0 1 RCIM_Cubic>,<at 0.5069 1>,<at 1 1 RCIM_Cubic>]}
       }}
      CurveIndex=Linked() {linkedVariable {name=NPC.FortniteNPC.TimeOfDay_0-1, type=NiagaraFloat}}
     ))
    ScaleA=true
    ScaleRGB=true
   ) {enabled=false, module=ScaleColor002}
   &ScaleColor(
    Scale Alpha=&FloatFromCurve(
     FloatCurve=DataInterface() {propertyValues {Curve {
        defaultValue=1
        keys [
         {interpMode=RCIM_Cubic, leaveTangent=-2.0674, tangentMode=RCTM_Break, time=0.3, value=1}
         {interpMode=RCIM_Cubic, tangentMode=RCTM_User, time=0.4, value=0.3}
         <at 1 0 RCIM_Cubic>
        ]
       }}}
     Scale Curve=&RandomRangeFloat(Minimum=0.8, Maximum=1)
    )
    Scale RGB=&Multiply_Vector(
     A=&MakeVectorFromLinearColorRGB(Color=&ColorFromCurve(ColorCurve=DataInterface() {propertyValues {
         AlphaCurve {keys [<at 0 1 RCIM_Cubic>,<at 1 0 RCIM_Cubic>]}
         BlueCurve {keys [<at 0 1 RCIM_Cubic>,<at 0.5069 1>,<at 1 1 RCIM_Cubic>]}
         GreenCurve {keys [<at 0 1 RCIM_Cubic>,<at 0.5069 1>,<at 1 1 RCIM_Cubic>]}
         RedCurve {keys [<at 0 1 RCIM_Cubic>,<at 0.5069 1>,<at 1 1 RCIM_Cubic>]}
        }}))
     B=&MakeVectorFromLinearColorRGB(Color=&FN_Select_Color_from_Time_Of_Day(
       Use Manual Time of Day Thresholds=true
       Morning=<rgba 1.5 0.5469 1.1714 1>
       Day=<rgba 1.6 1.025 1.4018 1>
       Evening=<rgba 1.8229 0.6456 1.7647 1>
       Night=<rgba 1.0716 0.7698 1.4 1>
       Day Threshold=17
       Evening Threshold=22
      ))
    )
    ScaleA=true
    ScaleRGB=true
   ) {module=ScaleColor003}
   &SubUVAnimation(
    Evaluation Type=::NewEnumerator0
    "Sprite Renderer "=DataInterface() {propertyValues {SpriteRenderer=&self:Water_Splashes_Translucent_0.NiagaraSpriteRendererProperties_1}}
    SubUV Animation Mode:ENiagara_SubUVLookupModeV2=::NewEnumerator2
    UseEndFrame=false
    UseStartFrame=false
   )
   &Drag(
    Drag=&FloatFromCurve(FloatCurve=DataInterface() {propertyValues {Curve {keys [<at 0 0>,<at 0.25 1>]}}}, Scale Curve=2)
    Use Linear Drag=true
    Use Rotational Drag=true
   )
   &GravityForce(Gravity=<0 0 -800>)
   &SolveForcesAndVelocity(Clamp Velocity=false, Limit Acceleration=false)
   &ScaleSpriteSizeBySpeed(
    Max Scale Factor=<4 4>
    Sample Scale Factor By Curve=true
    Scale Factor Curve=DataInterface() {propertyValues {Curve {keys [<at 0.7149 0.0353>,<at 0.7967 0.8118>]}}}
    Velocity Threshold=100
   )
   &DynamicMaterialParameters(
    Index 0 Param 1=&FloatFromCurve(FloatCurve=DataInterface() {propertyValues {Curve {keys [<at 0 1>,<at 1 0>]}}})
    Module.Index0Param0Write=true
    Module.Index0Param1Write=true
    Module.Index0Param2Write=true
    Module.Index0Param3Write=true
   )
   /SetVariables(Particles.Velocity2:Vector3f=&Multiply_VectorByFloat(Vector=$Particles.Velocity, Float=30)) {isSetParameters=true, module=SetVariables_#1}
  ]
  rendererClasses [&NiagaraSpriteRendererProperties]
  renderers [{rendererClass=&NiagaraSpriteRendererProperties}]
 }
 Water_Waves005 {
  simTarget=GPUComputeSim
  emitterUpdateScript [
   &EmitterState()
   &SpawnRate(SpawnRate=11, Use Spawn Probability=false)
  ]
  particleSpawnScript [
   &InitializeParticle(
    Color Mode=::NewEnumerator0
    Lifetime=&RandomRangeFloat(Minimum=2, Maximum=4)
    Mesh Scale Mode=::NewEnumerator1
    Mesh Uniform Scale Max=0.85
    Mesh Uniform Scale Min=0.6
    Sprite Size Mode=::NewEnumerator3
    Uniform Sprite Size=222
    UsePositionOffset=false
    Write Material Random=true
   )
   &self:CalcShapeRange(0-1RDM:NiagaraFloat=&RandomRangeFloat(Minimum=0, Maximum=1)) {module=CalcShapeRange}
   &AddVelocity(Module.Use Falloff Distance=false, Override Seed=false, Velocity Mode=::NewEnumerator1, Velocity Speed=630.725) {enabled=false}
   &ShapeLocation(
    Handle Radius=714.8043
    Large Radius=4740.2598
    MaskSpawnGroup=false
    Non Uniform Scale=<1 1 0.05>
    Offset=<0 0 85>
    Offset Mode=::NewEnumerator0
    Shape Primitive=::NewEnumerator3
    Torus Distribution Mode:ENiagaraTorusDistributionMode=::NewEnumerator1
    U Position=$Particles.CalcShapeRange.V
    UseOwnerScale=true
    V Position=&RandomRangeFloat(Minimum=0, Maximum=1)
   )
   &AddVelocity(Module.Use Falloff Distance=false, Origin Offset=<0 0 22>, Override Seed=false, Velocity Mode=::NewEnumerator1, Velocity Speed=-450.7053) {module=AddVelocity001}
   &InitialMeshOrientation(Mesh Orientation Mode=::NewEnumerator2, Orientation Vector=<-1 0 0>, Use Rotation Vector=false)
  ]
  particleUpdateScript [
   &ParticleState()
   &ScaleMeshSize(Scale Factor=<1.3 0.8 1>) {module=ScaleMeshSize001}
   &ScaleMeshSize(Scale Factor=&Multiply_Vector(
     A=&VectorFromFloat(Value=&FloatFromCurve(FloatCurve=DataInterface() {propertyValues {Curve {
          defaultValue=1
          keys [
           {arriveTangent=3.5072, interpMode=RCIM_Cubic, leaveTangent=3.5072, tangentMode=RCTM_User, time=0, value=0}
           {interpMode=RCIM_Cubic, tangentMode=RCTM_User, time=0.7, value=1}
          ]
         }}}))
     B=&MakeVector(
      X=&FloatFromCurve(
       FloatCurve=DataInterface() {propertyValues {Curve {
          defaultValue=1
          keys [{interpMode=RCIM_Cubic, tangentMode=RCTM_User, time=0, value=1.5}, {interpMode=RCIM_Cubic, tangentMode=RCTM_User, time=1, value=2}]
         }}}
       Scale Curve=&RandomRangeFloat(Minimum=0.8, Maximum=0.9)
      )
      Y=&FloatFromCurve(
       FloatCurve=DataInterface() {propertyValues {Curve {
          defaultValue=1
          keys [
           {arriveTangent=4.8818, interpMode=RCIM_Cubic, leaveTangent=0.4088, tangentMode=RCTM_Break, time=0, value=0.5}
           {arriveTangent=1.5, interpMode=RCIM_Cubic, leaveTangent=1.5, time=0.2065, value=1}
           {interpMode=RCIM_Cubic, tangentMode=RCTM_Break, time=1, value=2}
          ]
         }}}
       Scale Curve=&RandomRangeFloat(Minimum=1.3, Maximum=1.5)
      )
      Z=&FloatFromCurve(
       FloatCurve=DataInterface() {propertyValues {Curve {
          defaultValue=1
          keys [
           {arriveTangent=-0.7682, interpMode=RCIM_Cubic, leaveTangent=-0.7682, tangentMode=RCTM_User, time=0, value=1}
           {arriveTangent=-0.0262, interpMode=RCIM_Cubic, tangentMode=RCTM_Break, time=1, value=0.2}
          ]
         }}}
       Scale Curve=&RandomRangeFloat(Minimum=0.8, Maximum=1)
      )
     )
    )) {enabled=false, module=ScaleMeshSize003}
   &ScaleMeshSize(Scale Factor=&Multiply_Vector(
     A=&VectorFromFloat(Value=&FloatFromCurve(FloatCurve=DataInterface() {propertyValues {Curve {
          defaultValue=1
          keys [
           {arriveTangent=5.7121, interpMode=RCIM_Cubic, leaveTangent=5.7121, tangentMode=RCTM_User, time=0, value=0.3375}
           {interpMode=RCIM_Cubic, tangentMode=RCTM_User, time=0.7, value=1}
          ]
         }}}))
     B=&MakeVector(
      X=&FloatFromCurve(FloatCurve=DataInterface() {propertyValues {Curve {
          defaultValue=1
          keys [
           {interpMode=RCIM_Cubic, tangentMode=RCTM_User, time=0, value=0.2}
           {arriveTangent=4.4266, interpMode=RCIM_Cubic, tangentMode=RCTM_Break, time=1, value=2}
          ]
         }}})
      Y=&FloatFromCurve(
       FloatCurve=DataInterface() {propertyValues {Curve {
          defaultValue=1
          keys [
           {arriveTangent=4.8818, interpMode=RCIM_Cubic, leaveTangent=0.4088, tangentMode=RCTM_Break, time=0, value=0.5}
           {arriveTangent=1.5, interpMode=RCIM_Cubic, leaveTangent=1.5, time=0.2065, value=1}
           {interpMode=RCIM_Cubic, tangentMode=RCTM_Break, time=1, value=2}
          ]
         }}}
       Scale Curve=&RandomRangeFloat(Minimum=1.3, Maximum=1.5)
      )
      Z=&FloatFromCurve(
       FloatCurve=DataInterface() {propertyValues {Curve {
          defaultValue=1
          keys [
           {arriveTangent=-0.7682, interpMode=RCIM_Cubic, leaveTangent=-0.7682, tangentMode=RCTM_User, time=0, value=1}
           {arriveTangent=-0.0262, interpMode=RCIM_Cubic, tangentMode=RCTM_Break, time=1, value=0.2}
          ]
         }}}
       Scale Curve=&RandomRangeFloat(Minimum=0.2, Maximum=0.2)
      )
     )
    ))
   /SetVariables(Particles.GnomonCamera=222) {isSetParameters=true}
   &PointAttractionForce(Kill Within Radius=false, Module.Use Falloff=true, UseAttractorOffset=false) {enabled=false}
   &PointForce(Force Strength=-555.5, Module.Use Falloff=false, Module.UseLinearForce=false)
   &Drag(
    Drag=&FloatFromCurve(
     FloatCurve=DataInterface() {propertyValues {Curve {
        defaultValue=0
        keys [
         {interpMode=RCIM_Cubic, leaveTangent=-1.8821, tangentMode=RCTM_Break, time=0, value=1}
         {interpMode=RCIM_Cubic, tangentMode=RCTM_User, time=1, value=0}
        ]
       }}}
     Scale Curve=0.5
    )
    Use Linear Drag=true
    Use Rotational Drag=true
   ) {module=Drag001}
   &SolveForcesAndVelocity(Clamp Velocity=false, Limit Acceleration=false)
   /SetVariables(Particles.SpriteCameraDebug=222) {isSetParameters=true}
   &DynamicMaterialParameters(
    Index 0 Param 1=&FloatFromCurve(FloatCurve=DataInterface() {propertyValues {Curve {
        defaultValue=1
        keys [
         {interpMode=RCIM_Cubic, tangentMode=RCTM_User, time=0.4, value=1}
         {interpMode=RCIM_Cubic, tangentMode=RCTM_User, time=0.7347, value=0.3106}
         <at 0.9418 0.9123 RCIM_Cubic>
        ]
       }}})
    Index 0 Param 2=&FloatFromCurve(FloatCurve=DataInterface() {propertyValues {Curve {
        defaultValue=1
        keys [
         <at 0.021 1 RCIM_Cubic>
         {arriveTangent=-1.7271, interpMode=RCIM_Cubic, leaveTangent=0.0102, tangentMode=RCTM_Break, time=0.255, value=0}
         {interpMode=RCIM_Cubic, tangentMode=RCTM_User, time=0.6, value=0}
         {interpMode=RCIM_Cubic, tangentMode=RCTM_User, time=0.95, value=1}
        ]
       }}})
    Index 0 Param 3=&FloatFromCurve(
     FloatCurve=DataInterface() {propertyValues {Curve {
        defaultValue=1
        keys [{interpMode=RCIM_Cubic, tangentMode=RCTM_User, time=0, value=1}, {interpMode=RCIM_Cubic, tangentMode=RCTM_User, time=1, value=1.5}]
       }}}
     Scale Curve=&RandomRangeFloat(Minimum=0.75, Maximum=2)
    )
    Index 0 Param 4=&FloatFromCurve(FloatCurve=DataInterface() {propertyValues {Curve {
        defaultValue=1
        keys [{interpMode=RCIM_Cubic, tangentMode=RCTM_User, time=0, value=3}, {interpMode=RCIM_Cubic, tangentMode=RCTM_User, time=1, value=0.5}]
       }}})
    Module.Index0Float0Write=true
    Module.Index0Float1Write=true
    Module.Index0Float2Write=true
    Module.Index0Float3Write=true
   ) {enabled=false, module=DynamicMaterialParameters001}
   &DynamicMaterialParameters(
    Index 0 Param 1=&UniformAOrBFloat(
     A=&FloatFromCurve(FloatCurve=DataInterface() {propertyValues {Curve {
         defaultValue=1
         keys [
          {interpMode=RCIM_Cubic, tangentMode=RCTM_User, time=0.4, value=1}
          {interpMode=RCIM_Cubic, tangentMode=RCTM_User, time=0.7347, value=0.3106}
          <at 0.9418 0.9123 RCIM_Cubic>
         ]
        }}})
     Evaluation Type=::NewEnumerator0
     Recalculate Random Each Loop=true
     Override Seed=false
    )
    Index 0 Param 2=&FloatFromCurve(FloatCurve=DataInterface() {propertyValues {Curve {
        defaultValue=1
        keys [
         <at 0.021 1 RCIM_Cubic>
         {arriveTangent=-1.7271, interpMode=RCIM_Cubic, leaveTangent=0.0102, tangentMode=RCTM_Break, time=0.255, value=0}
         {interpMode=RCIM_Cubic, tangentMode=RCTM_User, time=0.6, value=0}
         {interpMode=RCIM_Cubic, tangentMode=RCTM_User, time=0.95, value=1}
        ]
       }}})
    Index 0 Param 3=&FloatFromCurve(
     FloatCurve=DataInterface() {propertyValues {Curve {
        defaultValue=1
        keys [{interpMode=RCIM_Cubic, tangentMode=RCTM_User, time=0, value=1}, {interpMode=RCIM_Cubic, tangentMode=RCTM_User, time=1, value=1.5}]
       }}}
     Scale Curve=&RandomRangeFloat(Minimum=0.75, Maximum=2)
    )
    Index 0 Param 4=&FloatFromCurve(
     FloatCurve=DataInterface() {propertyValues {Curve {
        defaultValue=1
        keys [{interpMode=RCIM_Cubic, tangentMode=RCTM_User, time=0, value=3}, {interpMode=RCIM_Cubic, tangentMode=RCTM_User, time=1, value=0.5}]
       }}}
     Scale Curve=&RandomRangeFloat(Minimum=0.6, Maximum=1)
    )
    Module.Index0Float0Write=true
    Module.Index0Float1Write=true
    Module.Index0Float2Write=true
    Module.Index0Float3Write=true
   ) {enabled=false}
   &DynamicMaterialParameters(
    Index 0 Param 1=&UniformAOrBFloat(
     A=&FloatFromCurve(FloatCurve=DataInterface() {propertyValues {Curve {
         defaultValue=1
         keys [
          {interpMode=RCIM_Cubic, tangentMode=RCTM_User, time=0.4, value=1}
          {interpMode=RCIM_Cubic, tangentMode=RCTM_User, time=0.7347, value=0.5}
          <at 0.9418 0.9123 RCIM_Cubic>
         ]
        }}})
     Distribution Weight=0.7
     Evaluation Type=::NewEnumerator0
     Recalculate Random Each Loop=true
     Override Seed=false
    )
    Index 0 Param 2=&FloatFromCurve(FloatCurve=DataInterface() {propertyValues {Curve {
        defaultValue=1
        keys [
         <at 0.021 1 RCIM_Cubic>
         {arriveTangent=-1.7271, interpMode=RCIM_Cubic, leaveTangent=0.0102, tangentMode=RCTM_Break, time=0.255, value=0}
         {interpMode=RCIM_Cubic, tangentMode=RCTM_User, time=0.6, value=0}
         {interpMode=RCIM_Cubic, tangentMode=RCTM_User, time=0.95, value=1}
        ]
       }}})
    Index 0 Param 3=&FloatFromCurve(
     FloatCurve=DataInterface() {propertyValues {Curve {
        defaultValue=1
        keys [
         {interpMode=RCIM_Cubic, tangentMode=RCTM_User, time=0, value=1}
         {arriveTangent=0.3873, interpMode=RCIM_Cubic, leaveTangent=0.3873, time=0.1035, value=0.5}
         {arriveTangent=1.1154, interpMode=RCIM_Cubic, leaveTangent=1.1154, time=0.2346, value=1.0909}
         {interpMode=RCIM_Cubic, tangentMode=RCTM_User, time=1, value=1.5}
        ]
       }}}
     Scale Curve=&RandomRangeFloat(Minimum=0.75, Maximum=1.5)
    )
    Index 0 Param 4=&UniformAOrBFloat(
     A=&FloatFromCurve(
      FloatCurve=DataInterface() {propertyValues {Curve {
         defaultValue=3
         keys [{interpMode=RCIM_Cubic, tangentMode=RCTM_User, time=0, value=0}, {interpMode=RCIM_Cubic, tangentMode=RCTM_User, time=0.4, value=2}]
        }}}
      Scale Curve=&RandomRangeFloat(Minimum=0.6, Maximum=1)
     )
     B=&FloatFromCurve(
      FloatCurve=DataInterface() {propertyValues {Curve {
         defaultValue=0
         keys [
          {interpMode=RCIM_Cubic, tangentMode=RCTM_User, time=0, value=5}
          {interpMode=RCIM_Cubic, tangentMode=RCTM_User, time=0.7, value=1}
          <at 1 0.3 RCIM_Cubic>
         ]
        }}}
      Scale Curve=&RandomRangeFloat(Minimum=0.8, Maximum=1)
     )
     Evaluation Type=::NewEnumerator0
     Recalculate Random Each Loop=true
     Override Seed=false
    )
    Module.Index0Float0Write=true
    Module.Index0Float1Write=true
    Module.Index0Float2Write=true
    Module.Index0Float3Write=true
   ) {module=DynamicMaterialParameters002}
   &ScaleColor(
    Scale Alpha=&FloatFromCurve(
     FloatCurve=DataInterface() {propertyValues {Curve {
        defaultValue=1
        keys [
         {tangentMode=RCTM_User, time=0, value=1}
         {arriveTangent=-0.0405, interpMode=RCIM_Cubic, leaveTangent=-0.0405, tangentMode=RCTM_User, time=0.6, value=1}
         {arriveTangent=-8.0089, interpMode=RCIM_Cubic, leaveTangent=-8.0089, tangentMode=RCTM_User, time=1, value=0}
        ]
       }}}
     Scale Curve=0.88
    )
    ScaleA=true
    ScaleRGB=false
   )
   &CameraOffset()
  ]
  rendererClasses [&NiagaraMeshRendererProperties, &NiagaraSpriteRendererProperties]
  renderers [
   {rendererClass=&NiagaraMeshRendererProperties, rendererIndex=0}
   {rendererClass=&NiagaraSpriteRendererProperties, rendererIndex=1}
  ]
 }
 Water_Waves_Body {
  emitterUpdateScript [
   &EmitterState()
   &SpawnRate(SpawnRate=10, Use Spawn Probability=false)
  ]
  particleSpawnScript [
   /SetVariables(Particles.RDM:NiagaraFloat=&RandomRangeFloat(Minimum=0, Maximum=1)) {isSetParameters=true}
   &self:CalcShapeRange(0-1RDM:NiagaraFloat=&RandomRangeFloat(Minimum=0, Maximum=1)) {module=CalcShapeRange}
   &InitializeParticle(
    Color Mode=::NewEnumerator0
    Lifetime=&RandomRangeFloat(Minimum=6, Maximum=8)
    Mesh Scale=&Multiply_VectorByFloat(
     Vector=&MakeVector(X=&Lerp_Float(A=0.7, B=1, Alpha=$Particles.RDM), Y=1, Z=1)
     Float=&Lerp_Float(A=1, B=1.6, Alpha=$Particles.RDM)
    )
    Mesh Scale Mode=::NewEnumerator4
    Sprite Size=<400 2448.3093>
    Sprite Size Mode=::NewEnumerator4
    UsePositionOffset=false
    Write Material Random=true
   )
   &ShapeLocation(
    Handle Radius=100
    Large Radius=5500
    MaskSpawnGroup=false
    Non Uniform Scale=<1 1 0>
    Offset=<0 0 140>
    Offset Mode=::NewEnumerator0
    Shape Primitive=::NewEnumerator3
    Torus Distribution Mode:ENiagaraTorusDistributionMode=::NewEnumerator1
    U Position=$Particles.CalcShapeRange.V
    UseOwnerScale=true
    V Position=&RandomRangeFloat(Minimum=0, Maximum=1)
   )
   &PointAttractionForce(
    Attraction Radius=1927956.125
    Attraction Strength=&RandomRangeFloat(Minimum=0.1, Maximum=0.15)
    Kill Within Radius=false
    Module.Use Falloff=false
    UseAttractorOffset=false
   )
   &ApplyInitialForces()
   &InitialMeshOrientation(Mesh Orientation Mode=::NewEnumerator7, Use Rotation Vector=false)
   &self:SetupOrientation() {module=SetupOrientation}
  ]
  particleUpdateScript [
   &ParticleState()
   /SetVariables(Particles.DistanceFade:NiagaraFloat=&FloatFromCurve(
     FloatCurve=DataInterface() {propertyValues {Curve {keys [<at 1000 0>,<at 2000 1>,<at 3500 1>,<at 4000 0>]}}}
     CurveIndex=&self:XYCamPos(Camera Query:NiagaraDataInterfaceCamera=DataInterface() {propertyValues {bRequireCurrentFrameData=false}})
    )) {isSetParameters=true}
   &ScaleMeshSize(Scale Factor=&VectorFromFloat(Value=$Particles.DistanceFade))
   &SolveForcesAndVelocity(Clamp Velocity=false, Limit Acceleration=false)
   &DynamicMaterialParameters(
    Index 0 Param 1=&Multiply_Float(
     A=&Multiply_Float(
      A=&Lerp_Float(A=70, B=105, Alpha=$Particles.RDM)
      B=&FloatFromCurve(FloatCurve=DataInterface() {propertyValues {Curve {
          defaultValue=1
          keys [
           {arriveTangent=0.2767, interpMode=RCIM_Cubic, leaveTangent=0.2767, tangentMode=RCTM_User, time=0, value=0.5}
           {arriveTangent=1.1705, interpMode=RCIM_Cubic, leaveTangent=1.1705, time=0.2439, value=1.1072}
           {arriveTangent=0.1186, interpMode=RCIM_Cubic, leaveTangent=0.1186, tangentMode=RCTM_User, time=1, value=1.6705}
          ]
         }}})
     )
     B=$Particles.DistanceFade
    )
    Index 0 Param 2=0.25
    Index 0 Param 4=&RandomRangeFloat(Minimum=0, Maximum=1000)
    Module.Index0Float0Write=true
    Module.Index0Float1Write=true
    Module.Index0Float2Write=true
    Module.Index0Float3Write=true
   )
   &ScaleColor(
    Scale Alpha=&FloatFromCurve(FloatCurve=DataInterface() {propertyValues {Curve {
        defaultValue=1
        keys [
         {interpMode=RCIM_Cubic, leaveTangent=4.9027, tangentMode=RCTM_Break, time=0, value=0}
         {interpMode=RCIM_Cubic, tangentMode=RCTM_User, time=0.5, value=1}
         {tangentMode=RCTM_User, time=1, value=0}
        ]
       }}})
    ScaleA=true
    ScaleRGB=true
   )
  ]
  rendererClasses [&NiagaraMeshRendererProperties, &NiagaraSpriteRendererProperties]
  renderers [
   {rendererClass=&NiagaraMeshRendererProperties, rendererIndex=0}
   {rendererClass=&NiagaraSpriteRendererProperties, rendererIndex=1}
  ]
 }
}
summary {
 systemName=NS_Env_ZeroPoint_Water
 emitters [
  {emitterName=Water_Drops001, simTarget=GPUComputeSim, rendererClasses [&NiagaraSpriteRendererProperties]}
  {
   emitterName=Water_Waves005
   simTarget=GPUComputeSim
   rendererClasses [&NiagaraMeshRendererProperties, &NiagaraSpriteRendererProperties]
  }
  {
   emitterName=Water_Waves_Body
   rendererClasses [&NiagaraMeshRendererProperties, &NiagaraSpriteRendererProperties]
  }
  {emitterName=Water_Drops, rendererClasses [&NiagaraSpriteRendererProperties]}
  {emitterName=Water_Mist, simTarget=GPUComputeSim, rendererClasses [&NiagaraSpriteRendererProperties]}
  {emitterName=Water_Splashes_Masked, simTarget=GPUComputeSim, rendererClasses [&NiagaraSpriteRendererProperties]}
  {emitterName=Water_Splashes, rendererClasses [&NiagaraSpriteRendererProperties]}
  {emitterName=Water_Splashes_Translucent, simTarget=GPUComputeSim, rendererClasses [&NiagaraSpriteRendererProperties]}
 ]
 userVariables [User.EffectScale {type=NiagaraFloat, defaultValue=1, description=""}]
}
